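# Translucent UI material for a scrim in the UIKit dialog system, drawn from a blurred backdrop.
# It samples Slate post buffer 0 through GetSlatePost0, routes the result as Blur RGB and slightly darkens it.
# A UseBlur scalar, on by default, blends between a near-one constant and one, likely for opacity, though the output connections are not shown.
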Material /FortUILibrary/Systems/Dialog/Materials/M_UIKit_Scrim
graph {
 properties {
  bUsedWithStaticMesh=false
  blendMode=BLEND_Translucent
  materialDomain=MD_UI
  pixelDepthOffsetMode=PDOM_Legacy
  usedShadingModels=Default Lit
 }
 nodes [
  LinearInterpolate(Alpha=ScalarParameter() {defaultValue=1, parameterName=UseBlur}) {constA=0.9}
  Multiply(A=NamedRerouteDeclaration(None=MaterialFunctionCall!("UVs (Optional)") {materialFunction=/Engine/Functions/UserInterface/GetSlatePost0}.LinearRGB) {name=Blur RGB, nodeColor=<rgba 0 0 0 1>}) {constB=0.8}
 ]
}
