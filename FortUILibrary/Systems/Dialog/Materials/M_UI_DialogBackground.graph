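# Translucent UI-domain material that, by its name, serves as a dialog background, built as a two-color radial gradient.
# UVs from the ScaleX output of MF_UI_UVScaling feed an exponential radial gradient with parameterized center, radius and density.
# That gradient drives two lerps between the RGB_A and RGB_B parameters, one for color and one for alpha, so each color carries its own transparency.
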Material /FortUILibrary/Systems/Dialog/Materials/M_UI_DialogBackground
graph {
 properties {
  bUsedWithStaticMesh=false
  blendMode=BLEND_Translucent
  materialDomain=MD_UI
  pixelDepthOffsetMode=PDOM_Legacy
  usedShadingModels=Default Lit
 }
 nodes [
  n1 = VectorParameter() {group=Color, parameterName=RGB_A, defaultValue=<rgba 1 0 1 1>}
  n2 = VectorParameter() {group=Color, parameterName=RGB_B, defaultValue=<rgba 1 0.25 1 1>}
  n3 = MaterialFunctionCall(
   UVs=MaterialFunctionCall() {materialFunction=&MF_UI_UVScaling}.ScaleX
   CenterPosition=MaterialFunctionCall(X=ScalarParameter() {defaultValue=0.5, parameterName=Center_X}, Y=ScalarParameter() {defaultValue=0.5, parameterName=Center_Y}) {materialFunction=&MakeFloat2}.Result
   Radius=ScalarParameter() {defaultValue=0.5, parameterName=Radius}
   Density=ScalarParameter() {defaultValue=2, parameterName=Density}
  ) {materialFunction=&RadialGradientExponential}
  LinearInterpolate(A=@n1.A, B=@n2.A, Alpha=@n3.RadialGradientExponential)
  LinearInterpolate(A=@n1.RGB, B=@n2.RGB, Alpha=@n3.RadialGradientExponential)
 ]
}
