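# UI-domain translucent material for images in the dialog system, with per-corner rounding.
# A texture parameter, defaulting to a soldier image, is sampled with UVs from MF_UI_UVScaling's Scaled To Fill output, driven by the texture's size.
# The sampled alpha is multiplied by a MF_UI_CoreUI_RoundedMask shape mask, controlled by a four-corner radius vector and a Radius In Pixels parameter.
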Material /FortUILibrary/Systems/Dialog/Materials/M_UI_DialogImage
graph {
 properties {
  bUsedWithStaticMesh=false
  blendMode=BLEND_Translucent
  materialDomain=MD_UI
  pixelDepthOffsetMode=PDOM_Legacy
  usedShadingModels=Default Lit
 }
 nodes [
  n1 = TextureObjectParameter() {group=Texture, parameterName=Texture, sortPriority=1, texture=/FortUILibrary/Systems/Items/Textures/T_UI_SoldierCommando_F}
  n2 = TextureProperty(Tex=@n1)
  n3 = VectorParameter() {
   group=Layout
   parameterName=Corner Radius
   sortPriority=2
   channelNames {a=Bottom Right, b=Top Right, g=Bottom Left, r=Top Left}
   defaultValue=<rgba 10 10 10 10>
  }
  Multiply(
   A=TextureSample(UVs=MaterialFunctionCall(TextureSize=@n2) {materialFunction=&MF_UI_UVScaling}.Scaled To Fill, Tex=@n1).A
   B=MaterialFunctionCall(
    Corner Radius=AppendVector(A=@n3, B=@n3)
    Radius in Pixels=ScalarParameter() {defaultValue=1, group=Layout, parameterName=Radius In Pixels, sliderMax=1, sortPriority=3}
   ) {materialFunction=&MF_UI_CoreUI_RoundedMask}.Shape Mask
  )
 ]
}
