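# UI-domain translucent material that draws a two-color linear gradient inside a rounded-corner shape, filed under the Dialog system.
# MF_UI_LinearGradient builds the blend factor from start, end, offset, rotation and center parameters, and it interpolates both RGB and alpha between Color and Color2.
# The blended alpha is multiplied by the MF_UI_CoreUI_RoundedMask shape mask, driven by a per-corner radius vector and a Radius In Pixels parameter.
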Material /FortUILibrary/Systems/Dialog/Materials/M_UI_LinearGradientWithRoundedCorners
graph {
 properties {
  bUsedWithStaticMesh=false
  blendMode=BLEND_Translucent
  materialDomain=MD_UI
  pixelDepthOffsetMode=PDOM_Legacy
  usedShadingModels=Default Lit
 }
 nodes [
  n1 = VectorParameter() {group=Color Settings, parameterName=Color, defaultValue=<rgba 0 0 0 1>}
  n2 = VectorParameter() {group=Color Settings, parameterName=Color2, defaultValue=<rgba 1 1 1 1>}
  n3 = VectorParameter() {
   group=Gradient Settings
   parameterName=GradientRotationCenter
   sortPriority=5
   channelNames {a="", b="", g=Center Y, r=Center X}
   defaultValue=<rgba 0.5 0.5 0 0>
  }
  n4 = MaterialFunctionCall(
   GradientStart=ScalarParameter() {group=Gradient Settings, parameterName=GradientStart, sortPriority=1}
   GradientEnd=ScalarParameter() {defaultValue=1, group=Gradient Settings, parameterName=GradientEnd, sortPriority=2}
   RotationAngle=ScalarParameter() {group=Gradient Settings, parameterName=GradientRotation, sortPriority=4}
   RotationCenter=@n3
   GradientOffset=ScalarParameter() {group=Gradient Settings, parameterName=GradientOffset, sortPriority=3}
  ) {materialFunction=&MF_UI_LinearGradient}
  LinearInterpolate(A=@n1.RGB, B=@n2.RGB, Alpha=@n4.Result)
  n5 = VectorParameter() {
   group=Layout
   parameterName=Corner Radius
   sortPriority=2
   channelNames {a=Bottom Right, b=Top Right, g=Bottom Left, r=Top Left}
   defaultValue=<rgba 10 10 10 10>
  }
  Multiply(
   A=LinearInterpolate(A=@n1.A, B=@n2.A, Alpha=@n4.Result)
   B=MaterialFunctionCall(
    Corner Radius=AppendVector(A=@n5, B=@n5)
    Radius in Pixels=ScalarParameter() {defaultValue=1, group=Layout, parameterName=Radius In Pixels, sliderMax=1, sortPriority=3}
   ) {materialFunction=&MF_UI_CoreUI_RoundedMask}.Shape Mask
  )
 ]
}
