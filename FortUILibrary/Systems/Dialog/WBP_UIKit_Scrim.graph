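# UIKit scrim widget for dialogs: a full-area blur overlay with fade-in/out animations and an optional invisible button that catches clicks.
# A post-buffer blur update feeds the MI_UIKit_Scrim image, followed by a capture widget named for restoring the buffer.
# Initialization applies click-catcher visibility and sets the material's UseBlur from EnableBlurForFrontEndWidgets config, defaulting to on.
# SetBlur only stores its value; SetDimColor and SetDimMaterial are unwired, targeting an absent DimBorder.
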WidgetBlueprint /FortUILibrary/Systems/Dialog/WBP_UIKit_Scrim
animations [
 Anim_FadeIn {
  binding_count=1
  length_seconds=0.5
  object=&self:Anim_FadeIn
  sequence {bindings [RootScrim {
     id=#1
     possessed_class=Border
     tracks [{
       type=MovieSceneColorTrack
       display_name=Content Color and Opacity
       sections [{
         type=MovieSceneColorSection
         end_frame=10
         start_frame=0
         channels [
          Color.R {type=MovieSceneScriptingFloatChannel}
          Color.G {type=MovieSceneScriptingFloatChannel}
          Color.B {type=MovieSceneScriptingFloatChannel}
          Color.A {type=MovieSceneScriptingFloatChannel, keys [[0,0],[10,1]]}
         ]
        }]
      }]
    }]}
 }
 Anim_FadeOut {
  binding_count=1
  length_seconds=0.25
  object=&self:Anim_FadeOut
  sequence {bindings [RootScrim {
     id=#2
     possessed_class=Border
     tracks [{
       type=MovieSceneColorTrack
       display_name=Content Color and Opacity
       sections [{
         type=MovieSceneColorSection
         end_frame=5
         start_frame=0
         channels [
          Color.R {type=MovieSceneScriptingFloatChannel}
          Color.G {type=MovieSceneScriptingFloatChannel}
          Color.B {type=MovieSceneScriptingFloatChannel}
          Color.A {type=MovieSceneScriptingFloatChannel, keys [[0,1],[5,0]]}
         ]
        }]
      }]
    }]}
 }
]
graphs [
 EventGraph {nodes [
   n1 = Event PreConstruct()
   > Branch(Condition=@n1.IsDesignTime)
   > Set RenderOpacity!(execute, RenderOpacity:float=0.2, self:Widget Object Reference=Get OverlayBlur() -> (OverlayBlur:Image Object Reference)) -> (then, Output_Get:float)
   Event OnInitialized()
   > SetBlur!(execute, self, Blur Strength:float=Get Blur() -> (Blur:double)) -> (then)
   > SetClickCatch!(execute, self, Enable:bool=Get UseClickCatcher() -> (UseClickCatcher:bool)) -> (then)
   > n2 = GetDynamicMaterial(self=Get OverlayBlur())
   n3 = GetConfigDataBool!(EntryName:String=EnableBlurForFrontEndWidgets) -> (bValueOut:bool, ReturnValue:bool)
   > SetScalarParameterValue(
    self=@n2
    ParameterName=UseBlur
    Value="To Float (Boolean)"(InBool=Select(Option 0:bool=true, Option 1:bool=@n3, Index=@n3.ReturnValue) -> (ReturnValue:bool))
   )
  ]}
 FadeIn {nodes [
   FadeIn() {is=title K2Node_FunctionEntry}
   > PlayAnimationForward(InAnimation=Get Anim_FadeIn() -> (Anim_FadeIn:Widget Animation Object Reference))
  ]}
 FadeOut {nodes [
   FadeOut() {is=title K2Node_FunctionEntry}
   > PlayAnimationForward(InAnimation=Get Anim_FadeOut() -> (Anim_FadeOut:Widget Animation Object Reference))
  ]}
 SetBlur {nodes [
   n1 = SetBlur!() -> (then, Blur Strength:float=10) {is=title K2Node_FunctionEntry}
   > Set Blur(Blur:double=@n1) -> (Output_Get:double)
  ]}
 SetClickCatch {nodes [
   n1 = SetClickCatch!() -> (then, Enable:bool=~) {is=title K2Node_FunctionEntry}
   > SetVisibility(
    self=Get ClickCatcher() -> (ClickCatcher:Button Object Reference)
    InVisibility=Select(Option 0:ESlateVisibility Enum=Collapsed, Option 1:ESlateVisibility Enum=Visible, Index=@n1) -> (ReturnValue:ESlateVisibility Enum)
   )
  ]}
 SetDimColor {nodes [
   SetBrushColor(self=Get DimBorder() -> (DimBorder:Border Object Reference))
   SetDimColor!() -> (then, Color:Linear Color Structure) {is=title K2Node_FunctionEntry}
  ]}
 SetDimMaterial {nodes [
   SetBrushFromMaterial(self:Border Object Reference=Get DimBorder())
   SetDimMaterial!() -> (then, Material Instance:Material Instance Object Reference) {is=title K2Node_FunctionEntry}
  ]}
]
variables [
 Blur {type=number}
 UseClickCatcher {type=boolean}
]
widgets {
 info {inheritedWidgetCount=0, parentClass=&UserWidget, rootWidgetClass=&Border, widgetCount=6}
 widgets [
  RootScrim {
   widgetClassPath=&Border
   widget=RootScrim
   properties {
    bShowEffectWhenDisabled=false
    background {
     bIsDynamicallyLoaded=false
     drawAs=NoDrawType
     imageType=NoImage
     mirroring=NoMirror
     resourceName=None
     resourceObject=null
     tiling=NoTile
     imageSize=<32 32>
     margin=<ltrb 0 0 0 0>
     outlineSettings {
      bUseBrushTransparency=false
      roundingType=HalfHeightRadius
      width=0
      color {colorUseRule=UseColor_Specified, specifiedColor=<rgba 0 0 0 0>}
      cornerRadii=<0 0 0 0>
     }
     tintColor {colorUseRule=UseColor_Specified, specifiedColor=<rgba 1 1 1 1>}
     uVRegion {bIsValid=false, max=<0 0>, min=<0 0>}
    }
    padding=<ltrb 0 0 0 0>
   }
  }
  Content {
   widgetClassPath=&GridPanel
   parent=RootScrim
   widget=Content
   properties {columnFill [1], rowFill [1]}
   slot {class=BorderSlot, properties {padding=<ltrb 0 0 0 0>}}
  }
  PostBufferUpdate_10 {
   widgetClassPath=/Script/UMG.PostBufferUpdate
   bIsVariable=true
   parent=Content
   widget=PostBufferUpdate_10
   properties {
    visibility=HitTestInvisible
    updateBufferInfos [{
      bufferToUpdate=ESlatePostRT_0
      postParamUpdater=&self:WidgetTree.PostBufferUpdate_10.PostBufferBlurUpdater_0
     }]
   }
   slot {class=GridSlot}
  }
  OverlayBlur {
   widgetClassPath=&Image
   bIsVariable=true
   parent=Content
   widget=OverlayBlur
   properties {
    visibility=SelfHitTestInvisible
    brush {
     bIsDynamicallyLoaded=false
     drawAs=Image
     imageType=FullColor
     mirroring=NoMirror
     resourceName=None
     resourceObject=/FortUILibrary/Systems/Dialog/MaterialInstances/MI_UIKit_Scrim
     tiling=NoTile
     imageSize=<32 32>
     margin=<ltrb 0 0 0 0>
     outlineSettings {
      bUseBrushTransparency=false
      roundingType=HalfHeightRadius
      width=0
      color {colorUseRule=UseColor_Specified, specifiedColor=<rgba 0 0 0 0>}
      cornerRadii=<0 0 0 0>
     }
     tintColor {colorUseRule=UseColor_Specified, specifiedColor=<rgba 1 1 1 1>}
     uVRegion {bIsValid=false, max=<0 0>, min=<0 0>}
    }
   }
   slot {class=GridSlot, properties {layer=1}}
  }
  PostBufferUpdate_UpdateBacktoDefault {
   widgetClassPath=&WBP_CaptureForPostBufferUpdate
   bIsVariable=true
   parent=Content
   widget=PostBufferUpdate_UpdateBacktoDefault
   slot {class=GridSlot, properties {layer=2}}
  }
  ClickCatcher {
   widgetClassPath=&Button
   parent=Content
   widget=ClickCatcher
   properties {
    visibility=Collapsed
    backgroundColor=<rgba 1 1 1 0>
    colorAndOpacity=<rgba 1 1 1 0>
    widgetStyle {
     clickedSlateSound {resourceObject=null}
     disabled {
      bIsDynamicallyLoaded=false
      drawAs=NoDrawType
      imageType=NoImage
      mirroring=NoMirror
      resourceName=None
      resourceObject=null
      tiling=NoTile
      imageSize=<0 0>
      margin=<ltrb 0 0 0 0>
      outlineSettings {
       bUseBrushTransparency=false
       roundingType=HalfHeightRadius
       width=0
       color {colorUseRule=UseColor_Specified, specifiedColor=<rgba 0 0 0 0>}
       cornerRadii=<0 0 0 0>
      }
      tintColor {colorUseRule=UseColor_Specified, specifiedColor=<rgba 1 1 1 0>}
      uVRegion {bIsValid=false, max=<0 0>, min=<0 0>}
     }
     disabledForeground {colorUseRule=UseColor_Specified, specifiedColor=<rgba 0.5271 0.5271 0.5271 1>}
     hovered {
      bIsDynamicallyLoaded=false
      drawAs=NoDrawType
      imageType=NoImage
      mirroring=NoMirror
      resourceName=None
      resourceObject=null
      tiling=NoTile
      imageSize=<32 32>
      margin=<ltrb 0 0 0 0>
      outlineSettings {
       bUseBrushTransparency=true
       roundingType=FixedRadius
       width=1
       color {colorUseRule=UseColor_Specified, specifiedColor=<rgba 0.7243 0.7243 0.7243 1>}
       cornerRadii=<4 4 4 4>
      }
      tintColor {colorUseRule=UseColor_Specified, specifiedColor=<rgba 1 1 1 0>}
      uVRegion {bIsValid=false, max=<0 0>, min=<0 0>}
     }
     hoveredForeground {colorUseRule=UseColor_Specified, specifiedColor=<rgba 0 1 1 1>}
     hoveredSlateSound {resourceObject=null}
     normal {
      bIsDynamicallyLoaded=false
      drawAs=NoDrawType
      imageType=NoImage
      mirroring=NoMirror
      resourceName=None
      resourceObject=null
      tiling=NoTile
      imageSize=<32 32>
      margin=<ltrb 0 0 0 0>
      outlineSettings {
       bUseBrushTransparency=true
       roundingType=FixedRadius
       width=1
       color {colorUseRule=UseColor_Specified, specifiedColor=<rgba 0.6951 0.6951 0.6951 1>}
       cornerRadii=<4 4 4 4>
      }
      tintColor {colorUseRule=UseColor_Specified, specifiedColor=<rgba 1 1 1 0>}
      uVRegion {bIsValid=false, max=<0 0>, min=<0 0>}
     }
     normalForeground {colorUseRule=UseColor_Specified, specifiedColor=<rgba 0 1 1 1>}
     normalPadding=<ltrb 0 0 0 0>
     pressed {
      bIsDynamicallyLoaded=false
      drawAs=NoDrawType
      imageType=NoImage
      mirroring=NoMirror
      resourceName=None
      resourceObject=null
      tiling=NoTile
      imageSize=<32 32>
      margin=<ltrb 0 0 0 0>
      outlineSettings {
       bUseBrushTransparency=true
       roundingType=FixedRadius
       width=1
       color {colorUseRule=UseColor_Specified, specifiedColor=<rgba 0.7243 0.7243 0.7243 1>}
       cornerRadii=<4 4 4 4>
      }
      tintColor {colorUseRule=UseColor_Specified, specifiedColor=<rgba 1 1 1 0>}
      uVRegion {bIsValid=false, max=<0 0>, min=<0 0>}
     }
     pressedForeground {colorUseRule=UseColor_Specified, specifiedColor=<rgba 0 1 1 1>}
     pressedPadding=<ltrb 0 0 0 0>
     pressedSlateSound {resourceObject=null}
    }
   }
   slot {class=GridSlot, properties {layer=4}}
  }
 ]
}
