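# Internal UIKit modal dialog widget that builds its layout from property structs and shows title, text, images, footer and buttons from a UIKitDialogViewModel.
# InitializeDialog adds optional parts via Try Add macros and places them in grid slots, with design-time viewmodel preview.
# Loud or Regular buttons join a button group whose clicks fire a dispatcher and can auto-deactivate the parent activatable widget.
# Activation and deactivation broadcast modal state, play animations and fade the scrim.
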WidgetBlueprint /FortUILibrary/Systems/Dialog/WBP_UIKit_Dialog_Internal
animations [
 Anim_OnActivate_Base {
  binding_count=1
  length_seconds=0.5
  object=&self:Anim_OnActivate_Base
  sequence {bindings [Main {
     id=#1
     possessed_class=GridPanel
     tracks [
      {
       type=MovieSceneFloatTrack
       display_name=Render Opacity
       sections [{type=MovieSceneFloatSection, end_frame=10, start_frame=0, channels [None {type=MovieSceneScriptingFloatChannel, keys [[4,0],[9.9997,1]]}]}]
      }
      {
       type=MovieScene2DTransformTrack
       sections [{
         type=MovieScene2DTransformSection
         end_frame=10
         start_frame=0
         channels [
          Scale.X {type=MovieSceneScriptingFloatChannel, keys [[4,1.05],[10,1]]}
          Scale.Y {type=MovieSceneScriptingFloatChannel, keys [[4,1.1],[10,1]]}
         ]
        }]
      }
     ]
    }]}
 }
 Anim_OnDeactivate_Base {
  binding_count=1
  length_seconds=0.25
  object=&self:Anim_OnDeactivate_Base
  sequence {bindings [Main {
     id=#2
     possessed_class=GridPanel
     tracks [{
       type=MovieSceneFloatTrack
       display_name=Render Opacity
       sections [{type=MovieSceneFloatSection, end_frame=5, start_frame=0, channels [None {type=MovieSceneScriptingFloatChannel, keys [[0,1],[5,0]]}]}]
      }]
    }]}
 }
]
graphs [
 ActivateDialog {nodes [
   ActivateDialog() {is=title K2Node_FunctionEntry}
   > Broadcast Event!(
    execute
    self:Gameplay Event Router Component Object Reference=GetGlobalEventRouter()
    EventData:BPS UIKit Dialog Modal State Change Structure=Make BPS UIKit Dialog Modal State Change!(bActive_4_#3:bool=true) -> (BPS_UIKitDialog_ModalStateChange:BPS UIKit Dialog Modal State Change Structure) {is=title K2Node_MakeStruct}
    EventContext:Object Reference=GetOwningPlayer()
   ) -> (then, ReturnValue:bool)
   > PlayAnimationForward(InAnimation=Get Anim_OnActivate_Base() -> (Anim_OnActivate_Base:Widget Animation Object Reference))
   > n1 = Get!(execute:Exec) -> (then:Exec, else:Exec, Scrim:WBP UIKit Scrim Object Reference)
   > FadeIn(self:WBP UIKit Scrim Object Reference=@n1)
  ]}
 AddDialogButtons {nodes [
   n1 = AddDialogButtons!() -> (then, Action Buttons:Array of UI Kit Dialog Button View Model Object References) {is=title K2Node_FunctionEntry}
   > Clear Buttons!(execute:Exec) -> (then:Exec) {is=title K2Node_MacroInstance}
   > n2 = For Each Loop(Array:Array of UI Kit Dialog Button View Model Object References=@n1) -> (Array Element:UI Kit Dialog Button View Model Object Reference)
   n3 = Break S UI Dialog Buttons!(S_UI_DialogButtons:"S UI Dialog Buttons Structure (by ref)"=Get Buttons() -> (Buttons:S UI Dialog Buttons Structure)) -> (
    GroupPlacement_71_#4:S UI Placement Structure
    SpaceBetweenButtons_62_#5:double=~
    EachButtonSize_90_#6:E_UI_ButtonSize Enum
    EachButtonMinWidth_77_#7:int
    EachButtonHorizontalFill_87_#8:bool=~
    CTAButtonsTypes_89_#9:Array of E_UI_ButtonVisualType Enums
    FocuseButton_85_#10:int
   ) {is=title K2Node_BreakStruct}
   > n4 = Branch(Condition=Is Valid Index!(TargetArray:Array of E_UI_ButtonVisualType Enums=@n3.CTAButtonsTypes_89_#9, IndexToTest:int=@n2.Array Index) -> (ReturnValue:bool))
   > n5 = Switch on E_UI_ButtonVisualType!(
    execute
    Selection:E_UI_ButtonVisualType Enum="Get (a ref)"(Array:Array of E_UI_ButtonVisualType Enums=@n3.CTAButtonsTypes_89_#9, Dimension 1=@n2.Array Index) -> (Output:"E_UI_ButtonVisualType Enum (by ref)")
   ) -> (NewEnumerator0:Exec, NewEnumerator1:Exec, NewEnumerator2:Exec) {is=title K2Node_SwitchEnum}
   n6 = Get DefaultHoldData() -> (DefaultHoldData:Common UIHold Data Class Reference)
   > n7 = Create Widget!(
    execute:Exec
    Class:User Widget Class Reference=/FortUILibrary/UIKit/Buttons/WBP_UIKit_Button_Loud.WBP_UIKit_Button_Loud_C
    OwningPlayer:Player Controller Object Reference
    Style:Common Button Style Class Reference=/FortUILibrary/UIKit/Buttons/Blocks/ButtonStyle_UIKit_Empty.ButtonStyle_UIKit_Empty_C
    bApplyAlphaOnDisable:bool=False
    bLocked:bool=False
    bSelectable:bool=False
    bShouldSelectUponReceivingFocus:bool=False
    bInteractableWhenSelected:bool=true
    bToggleable:bool=False
    bTriggerClickedAfterSelection:bool=False
    bDisplayInputActionWhenNotInteractable:bool=false
    bHideInputActionWithKeyboard:bool=False
    bShouldUseFallbackDefaultInputAction:bool=false
    bRequiresHold:bool=False
    HoldData:Common UIHold Data Class Reference=@n6
    ClickMethod:EButtonClickMethod Enum=DownAndUp
    InputPriority:int=0
    TriggeringInputAction:Data Table Row Handle Structure
   ) -> (then:Exec, ReturnValue:WBP UIKit Button Loud Object Reference)
   > SetButton!(
    execute
    self
    Button:"WBP UIKit Button CTA Base Object Reference (by ref)"=@n7
    Button Data:UI Kit Dialog Button View Model Object Reference=@n2
   ) -> (then)
   n8 = Create Widget(
    execute=[@n5.NewEnumerator1, @n4.else]
    Class=/FortUILibrary/UIKit/Buttons/WBP_UIKit_Button_Regular.WBP_UIKit_Button_Regular_C
    Style=&ButtonStyle_UIKit_Empty_C
    bApplyAlphaOnDisable=False
    bLocked=False
    bSelectable=False
    bShouldSelectUponReceivingFocus=False
    bInteractableWhenSelected=true
    bToggleable=False
    bTriggerClickedAfterSelection=False
    bDisplayInputActionWhenNotInteractable=false
    bHideInputActionWithKeyboard=False
    bShouldUseFallbackDefaultInputAction=false
    bRequiresHold=False
    HoldData=@n6
    ClickMethod=DownAndUp
    InputPriority=0
   ) -> (ReturnValue:WBP UIKit Button Regular Object Reference)
   > SetButton(Button=@n8, Button Data=@n2)
   n9 = Create Widget(
    execute=@n5.NewEnumerator2
    Class=&WBP_UIKit_Button_Regular_C
    Style=&ButtonStyle_UIKit_Empty_C
    bApplyAlphaOnDisable=False
    bLocked=False
    bSelectable=False
    bShouldSelectUponReceivingFocus=False
    bInteractableWhenSelected=true
    bToggleable=False
    bTriggerClickedAfterSelection=False
    bDisplayInputActionWhenNotInteractable=false
    bHideInputActionWithKeyboard=False
    bShouldUseFallbackDefaultInputAction=false
    bRequiresHold=False
    HoldData=@n6
    ClickMethod=DownAndUp
    InputPriority=0
   )
   > SetButtonType!(execute, self:WBP UIKit Button Regular Object Reference=@n9, ButtonType:E_UI_CTAButtonType Enum=NewEnumerator1) -> (then)
   > SetButton(Button=@n9, Button Data=@n2)
  ]}
 ApplyMobileFontSizeMultiplierToSlateFont {nodes [
   n1 = Inputs!() -> (SlateFontInfo:"Slate Font Info Structure (by ref)", MobileFontSizeMultiplier:double=~)
   n2 = Break Slate Font Info!(SlateFontInfo:"Slate Font Info Structure (by ref)"=@n1) -> (
    FontObject:Object Reference
    FontMaterial:Object Reference
    OutlineSettings:Font Outline Settings Structure
    TypefaceFontName:Name
    Size:float=~
    LetterSpacing:int
    SkewAmount:float=~
    bForceMonospaced:bool=~
    bMaterialIsStencil:bool=~
    MonospacedWidth:float=~
   ) {is=title K2Node_BreakStruct}
   n3 = Select(Option 0:double=1, Option 1:double=@n1.MobileFontSizeMultiplier, Index=IsUsingCompactUI()) -> (ReturnValue:double)
   Outputs!(ReturnValue:Slate Font Info Structure=MakeSlateFontInfo!(
     FontObject:Object Reference=@n2
     FontMaterial:Object Reference=@n2.FontMaterial
     OutlineSettings:Font Outline Settings Structure=@n2.OutlineSettings
     TypefaceFontName:Name=@n2.TypefaceFontName
     Size:float=float * float(A=@n2.Size, B=@n3) {is=title K2Node_PromotableOperator}
     LetterSpacing:int=@n2.LetterSpacing
     SkewAmount:float=@n2.SkewAmount
     bForceMonospaced:bool=@n2.bForceMonospaced
     bMaterialIsStencil:bool=@n2.bMaterialIsStencil
     MonospacedWidth:float=@n2.MonospacedWidth
    ) -> (ReturnValue:Slate Font Info Structure))
  ]}
 ClearButtons {nodes [
   Inputs!() -> (execute:Exec)
   > Clear(TargetArray:Array of WBP UIKit Button CTA Base Object References=Get ButtonWidgets() -> (ButtonWidgets:Array of WBP UIKit Button CTA Base Object References))
   > ClearChildren(self=Get ButtonsContainer() -> (ButtonsContainer:Horizontal Box Object Reference))
   > n1 = Get!(execute) -> (then, else, ButtonGroup:Common Button Group Base Object Reference)
   > n2 = RemoveAll(self:Common Widget Group Base Object Reference=@n1)
   Outputs!(then:Exec=[@n2.then, @n1.else])
  ]}
 DeactivateDialog {nodes [
   DeactivateDialog() {is=title K2Node_FunctionEntry}
   > Broadcast Event(
    self=GetGlobalEventRouter()
    EventData=Make BPS UIKit Dialog Modal State Change(bActive_4_#3=False)
    EventContext=GetOwningPlayer()
   )
   > PlayAnimationForward(InAnimation=Get Anim_OnDeactivate_Base() -> (Anim_OnDeactivate_Base:Widget Animation Object Reference))
   > n1 = Get!(execute) -> (then, else, Scrim:WBP UIKit Scrim Object Reference)
   > FadeOut(self:WBP UIKit Scrim Object Reference=@n1)
  ]}
 DeactivationRequested {nodes [
   n1 = DeactivationRequested!() -> (then, ShouldBeTrue:bool=~) {is=title K2Node_FunctionEntry}
   > Branch(Condition=@n1)
   > n2 = Get!(execute) -> (then, else, ParentActivatableDialog:Common Activatable Widget Object Reference)
   > DeactivateWidget(self=@n2)
  ]}
 DesignTimePreset {nodes [
   DesignTimePreset() {is=title K2Node_FunctionEntry}
   n1 = Get Slot_Main() -> (Slot_Main:Named Slot Object Reference)
   > SetVisibility(
    self=@n1
    InVisibility=Select(
     Option 0:ESlateVisibility Enum=Collapsed
     Option 1:ESlateVisibility Enum=SelfHitTestInvisible
     Index=HasAnyChildren!(self:Panel Widget Object Reference=@n1) -> (ReturnValue:bool)
    ) -> (ReturnValue:ESlateVisibility Enum)
   )
   n2 = Get Slot_Background() -> (Slot_Background:Named Slot Object Reference)
   > SetVisibility(self=@n2, InVisibility=Select(Option 0=Collapsed, Option 1=SelfHitTestInvisible, Index=HasAnyChildren(self=@n2)))
   n3 = Get Slot_Root() -> (Slot_Root:Named Slot Object Reference)
   > SetVisibility(self=@n3, InVisibility=Select(Option 0=Collapsed, Option 1=SelfHitTestInvisible, Index=HasAnyChildren(self=@n3)))
   n4 = Get Slot_Content() -> (Slot_Content:Named Slot Object Reference)
   > SetVisibility(self=@n4, InVisibility=Select(Option 0=Collapsed, Option 1=SelfHitTestInvisible, Index=HasAnyChildren(self=@n4)))
   n5 = Get Slot_SecondColumn() -> (Slot_SecondColumn:Named Slot Object Reference)
   > SetVisibility(self=@n5, InVisibility=Select(Option 0=Collapsed, Option 1=SelfHitTestInvisible, Index=HasAnyChildren(self=@n5)))
   n6 = Get DesignTimePreview_DialogVM() -> (DesignTimePreview_DialogVM:UI Kit Dialog Viewmodel Class Reference)
   > n7 = Construct Object from Class(Class=SelectClass!(
     A:Object Class Reference=@n6
     B:Object Class Reference=/Script/FortniteUI.UIKitDialogViewModel
     bSelectA:bool=IsValidClass!(Class:Object Class Reference=@n6) -> (ReturnValue:bool)
    ) -> (ReturnValue:Object Class Reference)) -> (ReturnValue:Object Reference)
   > n8 = Cast To UIKitDialogViewModel!(execute, Object=@n7) -> (then, CastFailed, AsUI Kit Dialog Viewmodel:UI Kit Dialog Viewmodel Object Reference) {is=title K2Node_DynamicCast}
   > SetDialogTitle!(execute, self, Title:Text=Get DialogTitle!(self:UI Kit Dialog Viewmodel Object Reference=@n8) -> (DialogTitle:Text)) -> (then)
   > SetDialogText!(execute, self, Text:Text=Get DialogInfoText!(self:UI Kit Dialog Viewmodel Object Reference=@n8) -> (DialogInfoText:Text)) -> (then)
   > SetDialogFooterText!(execute, self, FooterText:Text=Get DialogFooterText!(self:UI Kit Dialog Viewmodel Object Reference=@n8) -> (DialogFooterText:Text)) -> (then)
   > SetDialogImage!(
    execute
    self
    Texture:Texture 2D Object Reference=Get DialogImage!(self:UI Kit Dialog Viewmodel Object Reference=@n8) -> (DialogImage:Texture 2D Object Reference)
   ) -> (then)
   > SetDialogTitleImage!(
    execute
    self
    MaterialInstance:Material Instance Object Reference=Get DialogTitleImage!(self:UI Kit Dialog Viewmodel Object Reference=@n8) -> (DialogTitleImage:Material Instance Object Reference)
   ) -> (then)
   > SetDialogButtons!(
    execute
    self
    Action Buttons:Array of UI Kit Dialog Button View Model Object References=Get Buttons!(self:UI Kit Dialog Viewmodel Object Reference=@n8) -> (Buttons:Array of UI Kit Dialog Button View Model Object References)
   ) -> (then)
  ]}
 DetermineWidgetsUsedBasedOnVM {nodes [
   n1 = DetermineWidgetsUsedBasedOnVM!() -> (then, DialogVM:UI Kit Dialog Viewmodel Object Reference) {is=title K2Node_FunctionEntry}
   n2 = Get DialogTitle(self=@n1)
   > Set UseTitle(UseTitle:bool=NOT Boolean(A=TextIsEmpty(InText=@n2))) -> (Output_Get:bool)
   > Set UseImage(UseImage:bool=IsValid(Object=Get DialogImage(self=@n1))) -> (Output_Get:bool)
   > Set UseTitleImage(UseTitleImage:bool=IsValid(Object=Get DialogTitleImage(self=@n1))) -> (Output_Get:bool)
   n3 = Get DialogFooterText(self=@n1)
   > Set UseFooter(UseFooter:bool=NOT Boolean(A=TextIsEmpty(InText=@n3))) -> (Output_Get:bool)
  ]}
 DialogButtonClickedDispatcher {nodes [DialogButtonClickedDispatcher!() -> (then, ButtonIndex:int) {is=title K2Node_FunctionEntry}]}
 EventGraph {nodes [
   n1 = OnDialogButtonClicked!() -> (OutputDelegate, then, AssociatedButton:Common Button Base Object Reference, ButtonIndex:int) {is=title K2Node_CustomEvent}
   > Call Dialog Button Clicked Dispatcher!(execute:Exec, self:WBP UIKit Dialog Internal Object Reference, ButtonIndex:int=@n1.ButtonIndex) -> (then:Exec) {is=title K2Node_CallDelegate}
   > Try Auto Deactivate!(execute:Exec, ButtonIndex:int=@n1.ButtonIndex) -> (Out:Exec) {is=title K2Node_MacroInstance}
   n2 = Event PreConstruct()
   > n3 = Branch(Condition=IsValid(Object=Get ButtonGroup() -> (ButtonGroup:Common Button Group Base Object Reference)))
   InitializeDialog(execute=[@n3.then, @n5.then])
   > Branch(Condition=@n2.IsDesignTime)
   > DesignTimePreset() {is=title K2Node_CallFunction}
   n4 = Construct Object from Class!(execute=@n3.else, Class=/Script/CommonUI.CommonButtonGroupBase, self, bSelectionRequired:bool=False) -> (then, ReturnValue:Common Button Group Base Object Reference)
   > n5 = Set ButtonGroup(ButtonGroup:Common Button Group Base Object Reference=@n4) -> (Output_Get:Common Button Group Base Object Reference)
   Event Construct()
   > n6 = FindParentWidgetOfType!(execute, StartingWidget:Widget Object Reference=Self-Reference(), Type:Widget Class Reference=&CommonActivatableWidget) -> (then, ReturnValue:Common Activatable Widget Object Reference)
   > n7 = Set ParentActivatableDialog(ParentActivatableDialog:Common Activatable Widget Object Reference=@n6) -> (Output_Get:Common Activatable Widget Object Reference)
   n8 = Create Event()
   > Bind Event to On Widget Activated(self:Common Activatable Widget Object Reference=@n7, Delegate=@n8) {is=title K2Node_AddDelegate}
   n9 = Create Event()
   > Bind Event to On Widget Deactivated(self:Common Activatable Widget Object Reference=@n7, Delegate=@n9) {is=title K2Node_AddDelegate}
   n10 = Create Event()
   n11 = Get ButtonGroup()
   > Bind Event to On Button Base Clicked(self:Common Button Group Base Object Reference=@n11, Delegate=@n10) {is=title K2Node_AddDelegate}
   Event Destruct()
   > n12 = Get()
   > Unbind Event from On Widget Activated(self:Common Activatable Widget Object Reference=@n12, Delegate=@n8) {is=title K2Node_RemoveDelegate}
   > Unbind Event from On Widget Deactivated(self:Common Activatable Widget Object Reference=@n12, Delegate=@n9) {is=title K2Node_RemoveDelegate}
   > Set ParentActivatableDialog()
   > Unbind Event from On Button Base Clicked(self:Common Button Group Base Object Reference=@n11, Delegate=@n10) {is=title K2Node_RemoveDelegate}
  ]}
 InitializeDialog {nodes [
   InitializeDialog() {is=title K2Node_FunctionEntry}
   > SetSizeBox!(execute, self, Data:"S UI Size Box Structure (by ref)"=Get DialogSIze() -> (DialogSIze:S UI Size Box Structure)) -> (then)
   > SetPlacement!(
    execute
    self
    Widget:Widget Object Reference=Get SizeBox_Dialog() -> (SizeBox_Dialog:Size Box Object Reference)
    Placement:S UI Placement Structure=Get DialogPlacement() -> (DialogPlacement:S UI Placement Structure)
   ) -> (then)
   > Try Add Scrim!(execute:Exec) -> (then:Exec) {is=title K2Node_MacroInstance}
   > SetScrim!(
    execute
    self
    Blur Strength:float=Get Blur() -> (Blur:double)
    Color:Linear Color Structure=Get DimColor() -> (DimColor:Linear Color Structure)
    Material Instance:Material Instance Object Reference=Get DimMaterial() -> (DimMaterial:Material Instance Object Reference)
    Enable:bool=Get UseClickCatch() -> (UseClickCatch:bool)
   ) -> (then)
   > Try Add Title!(execute:Exec) -> (then:Exec) {is=title K2Node_MacroInstance}
   > SetTitleProperties!(execute, self, Data:S UI Text Structure=Get TitleProperties() -> (TitleProperties:S UI Text Structure)) -> (then)
   > Try Add Text!(execute:Exec) -> (then:Exec) {is=title K2Node_MacroInstance}
   > SetTextProperties!(execute, self, Data:S UI Text Structure=Get TextProperties() -> (TextProperties:S UI Text Structure)) -> (then)
   > Try Add Image!(execute:Exec) -> (then:Exec) {is=title K2Node_MacroInstance}
   > SetImageProperties!(execute, self, Data:"S UI Image Structure (by ref)"=Get ImageProperties() -> (ImageProperties:S UI Image Structure)) -> (then)
   > Try Add Footer!(execute:Exec) -> (then:Exec) {is=title K2Node_MacroInstance}
   > SetFooterTextProperties!(execute, self, Data:S UI Text Structure=Get FooterTextProperties() -> (FooterTextProperties:S UI Text Structure)) -> (then)
   > Try Add Backplate!(execute:Exec) -> (then:Exec) {is=title K2Node_MacroInstance}
   > SetBackplateProperties!(execute, self, Data:"S UI Backplate Structure (by ref)"=Get BackplateProperties() -> (BackplateProperties:S UI Backplate Structure)) -> (then)
   > Try Add Background!(execute:Exec) -> (then:Exec) {is=title K2Node_MacroInstance}
   > SetBackgroundProperties!(execute, self, Data:"S UI Image Structure (by ref)"=Get BackgroundProperties() -> (BackgroundProperties:S UI Image Structure)) -> (then)
   > SetPlacement(
    Widget=Get ButtonsContainer()
    Placement=Break S UI Dialog Buttons!(S_UI_DialogButtons=Get Buttons!() -> (Buttons:S UI Dialog Buttons Structure)) -> (GroupPlacement_71_#4)
   )
  ]}
 PresetForDesignTime {nodes [
   Inputs()
   n1 = Get Slot_Main()
   > SetVisibility(self=@n1, InVisibility=Select(Option 0=Collapsed, Option 1=SelfHitTestInvisible, Index=HasAnyChildren(self=@n1)))
   n2 = Get Slot_Background()
   > SetVisibility(self=@n2, InVisibility=Select(Option 0=Collapsed, Option 1=SelfHitTestInvisible, Index=HasAnyChildren(self=@n2)))
   n3 = Get Slot_Root()
   > SetVisibility(self=@n3, InVisibility=Select(Option 0=Collapsed, Option 1=SelfHitTestInvisible, Index=HasAnyChildren(self=@n3)))
   n4 = Get Slot_Content()
   > SetVisibility(self=@n4, InVisibility=Select(Option 0=Collapsed, Option 1=SelfHitTestInvisible, Index=HasAnyChildren(self=@n4)))
   n5 = Get Slot_SecondColumn()
   > SetVisibility(self=@n5, InVisibility=Select(Option 0=Collapsed, Option 1=SelfHitTestInvisible, Index=HasAnyChildren(self=@n5)))
   n6 = Get DesignTimePreview_DialogVM()
   > n7 = Construct Object from Class!(execute, Class=SelectClass(A=@n6, B=&UIKitDialogViewModel, bSelectA=IsValidClass(Class=@n6)), self) -> (then, ReturnValue:Object Reference)
   > n8 = Cast To UIKitDialogViewModel(Object=@n7)
   > SetDialogTitle(Title=Get DialogTitle(self=@n8))
   > SetDialogText(Text=Get DialogInfoText(self=@n8))
   > SetDialogFooterText(FooterText=Get DialogFooterText(self=@n8))
   > SetDialogImage(Texture=Get DialogImage(self=@n8))
   > SetDialogTitleImage(MaterialInstance=Get DialogTitleImage(self=@n8))
   > SetDialogButtons(Action Buttons=Get Buttons!(self:UI Kit Dialog Viewmodel Object Reference=@n8) -> (Buttons:Array of UI Kit Dialog Button View Model Object References))
   > Outputs()
  ]}
 Remove {nodes [
   n1 = Inputs!() -> (execute, Widget:Widget Object Reference)
   > n2 = Branch(Condition=IsValid(Object=@n1))
   > n3 = RemoveFromParent(self=@n1)
   Outputs(then=[@n3.then, @n2.else])
  ]}
 SetBackgroundProperties {nodes [
   n1 = SetBackgroundProperties!() -> (then, Data:"S UI Image Structure (by ref)") {is=title K2Node_FunctionEntry}
   n2 = Break S UI Image!(S_UI_Image:"S UI Image Structure (by ref)"=@n1) -> (
    Placement_67_#4:S UI Placement Structure
    ImageSize_72_#11:Vector 2D Structure
    Material_75_#12:Material Instance Object Reference
   ) {is=title K2Node_BreakStruct}
   > SetPlacement(Widget=Get Slot_Background(), Placement=@n2)
   > n3 = Get!(execute) -> (then, else, ContentBackground:Image Object Reference)
   > SetDesiredSizeOverride!(execute, self:Image Object Reference=@n3, DesiredSize:Vector 2D Structure=@n2.ImageSize_72_#11) -> (then)
   > SetBrushFromMaterial(self=@n3, Material=@n2.Material_75_#12)
  ]}
 SetBackplateProperties {nodes [
   n1 = SetBackplateProperties!() -> (then, Data:"S UI Backplate Structure (by ref)") {is=title K2Node_FunctionEntry}
   > n2 = Get!(execute) -> (then, else, ContentBackplate:WBP UIKit Backplate Base Object Reference)
   n3 = Break S UI Backplate!(S_UI_Backplate:"S UI Backplate Structure (by ref)"=@n1) -> (
    Placement_84_#13:S UI Placement Structure
    Brightness_85_#4:E_UI_BackplateBrightness Enum
    CornerRadius_86_#11:E_UI_BackplateCornerRadius Enum
    Blur_81_#12:double=~
   ) {is=title K2Node_BreakStruct}
   > SetPlacement(Widget=@n2, Placement=@n3)
   > SetBrightness!(
    execute
    self:WBP UIKit Backplate Base Object Reference=@n2
    Brightness:E_UI_BackplateBrightness Enum=@n3.Brightness_85_#4
   ) -> (then)
   > SetCornerRadius!(
    execute
    self:WBP UIKit Backplate Base Object Reference=@n2
    Corner Radius:E_UI_BackplateCornerRadius Enum=@n3.CornerRadius_86_#11
   ) -> (then)
  ]}
 SetButton {nodes [
   n1 = SetButton!() -> (then, Button:"WBP UIKit Button CTA Base Object Reference (by ref)", Button Data:UI Kit Dialog Button View Model Object Reference) {is=title K2Node_FunctionEntry}
   > n2 = Set ButtonReference(ButtonReference:WBP UIKit Button CTA Base Object Reference=@n1) -> (Output_Get:WBP UIKit Button CTA Base Object Reference)
   > SetText!(
    execute
    self:Fort CTAButton Object Reference=@n2
    ButtonText:"Text (by ref)"=Get Text(self:UI Kit Dialog Button View Model Object Reference=@n1.Button Data)
   ) -> (then)
   > SetIsFocusable!(execute, self:Common Button Base Object Reference=@n2, bInIsFocusable:bool=true) -> (then)
   > n3 = Branch(Condition=IsValid(Object=Get CommonInputSubsystem()))
   n4 = Get DataTable!(self:UI Kit Dialog Button View Model Object Reference=@n1.Button Data) -> (DataTable:Data Table Soft Object Reference)
   n5 = IsAssetNull!(Asset:Soft Object Reference=@n4) -> (ReturnValue:bool)
   n6 = Resolve Soft Reference!(Input:Data Table Soft Object Reference=@n4) -> (Output:Data Table Object Reference)
   n7 = Get DefaultInputActionDataTable() -> (DefaultInputActionDataTable:Data Table Object Reference)
   > n8 = SetTriggeringInputAction!(
    execute
    self:Common Button Base Object Reference=@n2
    InputActionRow:"Data Table Row Handle Structure (by ref)"=Make Data Table Row Handle!(
     DataTable:Data Table Object Reference=Select(Option 0:Data Table Object Reference=@n6, Option 1:Data Table Object Reference=@n7, Index=@n5) -> (ReturnValue:Data Table Object Reference)
     RowName:Name=Get InputAction!(self:UI Kit Dialog Button View Model Object Reference=@n1.Button Data) -> (InputAction:Name=None)
    ) -> (DataTableRowHandle:Data Table Row Handle Structure) {is=title K2Node_MakeStruct}
   ) -> (then)
   AddChild(execute=[@n8.then, @n3.else], self=Get ButtonsContainer(), Content=@n2)
   > Add(
    TargetArray:Array of WBP UIKit Button CTA Base Object References=Get ButtonWidgets()
    NewItem:"WBP UIKit Button CTA Base Object Reference (by ref)"=@n2
   ) {is=title K2Node_CallArrayFunction}
   > AddWidget!(execute, self:Common Widget Group Base Object Reference=Get ButtonGroup(), InWidget:Widget Object Reference=@n2) -> (then)
  ]}
 SetButtonEnabled {nodes [
   n1 = SetButtonEnabled!() -> (then, ButtonIndex:int, IsButtonEnabled:bool=~) {is=title K2Node_FunctionEntry}
   > SetIsEnabled(
    self="Get (a copy)"(Array:Array of WBP UIKit Button CTA Base Object References=Get ButtonWidgets(), Dimension 1=@n1) -> (Output:WBP UIKit Button CTA Base Object Reference)
    bInIsEnabled=@n1.IsButtonEnabled
   )
  ]}
 SetButtonVisibility {nodes [
   n1 = SetButtonVisibility!() -> (then, ButtonIndex:int, Visibility:ESlateVisibility Enum=Visible) {is=title K2Node_FunctionEntry}
   > SetVisibility(self="Get (a copy)"(Array=Get ButtonWidgets(), Dimension 1=@n1), InVisibility=@n1.Visibility)
   > Set SkipButtonSpacing(SkipButtonSpacing:bool=true) -> (Output_Get:bool)
   > n2 = For Each Loop(Array:Array of WBP UIKit Button CTA Base Object References=Get ButtonWidgets()) -> (Array Element:WBP UIKit Button CTA Base Object Reference)
   > n3 = Set ArrayWidget(ArrayWidget:WBP UIKit Button CTA Base Object Reference=@n2) -> (Output_Get:WBP UIKit Button CTA Base Object Reference)
   > Branch(Condition=IsVisible(self=@n3))
   n4 = Break S UI Dialog Buttons!(S_UI_DialogButtons=Get Buttons!() -> (Buttons:S UI Dialog Buttons Structure)) -> (SpaceBetweenButtons_62_#5:double=~)
   n5 = Get SkipButtonSpacing() -> (SkipButtonSpacing:bool)
   > SetPadding!(
    execute
    self:User Widget Object Reference=Get ArrayWidget() -> (ArrayWidget:WBP UIKit Button CTA Base Object Reference)
    InPadding:Margin Structure=Make Margin(Left=SelectFloat(A=0, B=@n4, bPickA=@n5), Top=0, Right=0, Bottom=0)
   ) -> (then)
   > Set SkipButtonSpacing(SkipButtonSpacing=false)
  ]}
 SetButtonsProperties {nodes [
   SetButtonsProperties() {is=title K2Node_FunctionEntry}
   > Set SkipButtonSpacing(SkipButtonSpacing=true)
   > n1 = For Each Loop(Array=Get ButtonWidgets())
   > n2 = Set ArrayWidget(ArrayWidget=@n1)
   > Branch(Condition=IsVisible(self=@n2))
   n3 = Break S UI Dialog Buttons(S_UI_DialogButtons=Get Buttons()) -> (SpaceBetweenButtons_62_#5=~)
   n4 = Get SkipButtonSpacing()
   > SetPadding(self=Get ArrayWidget(), InPadding=Make Margin(Left=SelectFloat(A=0, B=@n3, bPickA=@n4), Top=0, Right=0, Bottom=0))
   > Set SkipButtonSpacing(SkipButtonSpacing=false)
   n5 = Break S UI Dialog Buttons!(S_UI_DialogButtons=Get Buttons()) -> (EachButtonSize_90_#6:E_UI_ButtonSize Enum, EachButtonMinWidth_77_#7:int)
   n6 = Get ArrayWidget()
   > SetButtonSize!(execute, self:WBP UIKit Button CTA Base Object Reference=@n6, ButtonSize:E_UI_ButtonSize Enum=@n5) -> (then)
   > n7 = Set MinWidthOverride!(
    execute
    MinWidthOverride:int=@n5.EachButtonMinWidth_77_#7
    self:WBP UIKit Button CTA Base Object Reference=@n6
   ) -> (then, Output_Get:int=0)
   > Set MobileMinWidth!(execute, MobileMinWidth:int=@n7, self:WBP UIKit Button CTA Base Object Reference=@n6) -> (then, Output_Get:int=0)
   > UpdateButtonSIze(self:WBP UIKit Button CTA Base Object Reference=@n6)
   > n8 = Cast To HorizontalBoxSlot!(execute, Object=Get Slot!(self:Widget Object Reference=Get ArrayWidget()) -> (Slot:Panel Slot Object Reference)) -> (then, CastFailed, AsHorizontal Box Slot:Horizontal Box Slot Object Reference) {is=title K2Node_DynamicCast}
   n9 = Get Buttons()
   > SetSize!(
    execute
    self:Horizontal Box Slot Object Reference=@n8
    InSize:Slate Child Size Structure=Select(
     Option 0:Slate Child Size Structure=Make Slate Child Size!(Value:float=1, SizeRule:ESlateSizeRule Enum=Automatic) -> (SlateChildSize:Slate Child Size Structure) {is=title K2Node_MakeStruct}
     Option 1:Slate Child Size Structure=Make Slate Child Size(Value=1, SizeRule=Fill)
     Index=Break S UI Dialog Buttons!(S_UI_DialogButtons=@n9) -> (EachButtonHorizontalFill_87_#8:bool=~)
    ) -> (ReturnValue:Slate Child Size Structure)
   ) -> (then)
  ]}
 SetCommonTextProperties {nodes [
   SetCommonTextProperties!() -> (then, S_UI_Text:"S UI Text Structure (by ref)", TextWidget:Common Text Object Reference, PlacementWidget:Widget Object Reference) {is=title K2Node_FunctionEntry}
   > Get!(execute) -> (then, else, TextWidget:Common Text Object Reference)
   n1 = Break S UI Text!(S_UI_Text:"S UI Text Structure (by ref)"=Get S_UI_Text() -> (S_UI_Text:"S UI Text Structure (by ref)")) -> (
    Placement_67_#4:S UI Placement Structure
    Font_45_#14:Slate Font Info Structure
    Color_65_#15:Slate Color Structure
    TransformPolicy_49_#16:ETextTransformPolicy Enum
    Justification_52_#17:ETextJustify Enum
    LineHeightPercentage_55_#18:double=~
    AutoWrapText_58_#19:bool=~
    WrapTextAt_68_#5:int
    MobileFontSizeMultiplier_73_#20:double=~
   ) {is=title K2Node_BreakStruct}
   > SetPlacement(Widget=Get PlacementWidget() -> (PlacementWidget:Widget Object Reference), Placement=@n1)
   > SetFont(self=Get TextWidget() -> (TextWidget:Common Text Object Reference), InFontInfo=@n1.Font_45_#14)
   > SetColorAndOpacity(self=Get TextWidget(), InColorAndOpacity=@n1.Color_65_#15)
   > Set TextTransformPolicy!(
    execute
    TextTransformPolicy:ETextTransformPolicy Enum=@n1.TransformPolicy_49_#16
    self:Text Object Reference=Get TextWidget()
   ) -> (then, Output_Get:ETextTransformPolicy Enum=None)
   > Set Justification!(
    execute
    Justification:ETextJustify Enum=@n1.Justification_52_#17
    self:Text Layout Widget Object Reference=Get TextWidget()
   ) -> (then, Output_Get:ETextJustify Enum=Left)
   > SetLineHeightPercentage!(
    execute
    self:Common Text Object Reference=Get TextWidget()
    InLineHeightPercentage:float=@n1.LineHeightPercentage_55_#18
   ) -> (then)
   > SetAutoWrapText!(execute, self:Text Object Reference=Get TextWidget(), InAutoTextWrap:bool=@n1.AutoWrapText_58_#19) -> (then)
   > SetWrapTextWidth!(execute, self:Common Text Object Reference=Get TextWidget(), InWrapTextAt:int=@n1.WrapTextAt_68_#5) -> (then)
   > SetMobileFontSizeMultiplier!(
    execute
    self:Common Text Object Reference=Get TextWidget()
    InMobileFontSizeMultiplier:float=@n1.MobileFontSizeMultiplier_73_#20
   ) -> (then)
  ]}
 SetDialogButtons {nodes [
   n1 = SetDialogButtons!() -> (then, Action Buttons:Array of UI Kit Dialog Button View Model Object References) {is=title K2Node_FunctionEntry}
   > AddDialogButtons!(execute, self, Action Buttons:Array of UI Kit Dialog Button View Model Object References=@n1) -> (then) {is=title K2Node_CallFunction}
   > SetButtonsProperties() {is=title K2Node_CallFunction}
  ]}
 SetDialogData {nodes [
   n1 = SetDialogData!() -> (then, Data:UI Kit Dialog Viewmodel Object Reference) {is=title K2Node_FunctionEntry}
   > InitializeDialog() {is=title K2Node_CallFunction}
   > SetUIKitDialogViewModel!(execute, self, Viewmodel:UI Kit Dialog Viewmodel Object Reference=@n1) -> (then)
   > SetButtonsProperties()
  ]}
 SetDialogFooterText {nodes [
   n1 = SetDialogFooterText!() -> (then, FooterText:Text) {is=title K2Node_FunctionEntry}
   > n2 = Get!(execute) -> (then, else, FooterText:Common Text Object Reference)
   > "SetText (Text)"(self=@n2, InText=@n1)
  ]}
 SetDialogFooterVisibility {nodes [
   n1 = SetDialogFooterVisibility!() -> (then, IsFooterVisibile:bool=~) {is=title K2Node_FunctionEntry}
   > n2 = Get()
   > SetVisibility(
    self=@n2
    InVisibility=Select(Option 0:ESlateVisibility Enum=Collapsed, Option 1:ESlateVisibility Enum=SelfHitTestInvisible, Index=@n1) -> (ReturnValue:ESlateVisibility Enum)
   )
  ]}
 SetDialogImage {nodes [
   n1 = SetDialogImage!() -> (then, Texture:Texture 2D Object Reference) {is=title K2Node_FunctionEntry}
   > Branch(Condition=IsValid(Object=@n1))
   > n2 = Get!(execute) -> (then, else, ContentImage:Image Object Reference)
   > SetBrushFromMaterial(self=@n2, Material=Break S UI Image(S_UI_Image=Get ImageProperties()).Material_75_#12)
   > n3 = GetDynamicMaterial(self=@n2)
   > n4 = Set ImageMID(ImageMID:Material Instance Dynamic Object Reference=@n3) -> (Output_Get:Material Instance Dynamic Object Reference)
   > SetTextureParameterValue(self=@n4, ParameterName=Texture, Value=@n1)
   n5 = Break S UI Backplate Corner Radius Values!(S_UI_BackplateCornerRadiusValues:"S UI Backplate Corner Radius Values Structure (by ref)"=Get CornerRadiusValues() -> (CornerRadiusValues:S UI Backplate Corner Radius Values Structure)) -> (
    Backplate_11_#21:double=~
    InteractiveComponent_12_#22:double=~
    Button_13_#23:double=~
    Tab_14_#24:double=~
    Circle_15_#25:double=~
   ) {is=title K2Node_BreakStruct}
   > SetVectorParameterValue(
    self=@n4
    ParameterName=Corner Radius
    Value=Make Linear Color!(R:float=0, G:float=0, B:float=@n5, A:float=@n5) -> (LinearColor:Linear Color Structure) {is=title K2Node_MakeStruct}
   )
  ]}
 SetDialogText {nodes [
   n1 = SetDialogText!() -> (then, Text:Text) {is=title K2Node_FunctionEntry}
   > n2 = Switch on E_UI_DialogTextType!(execute, Selection:E_UI_DialogTextType Enum=Get UseText() -> (UseText:E_UI_DialogTextType Enum=NewEnumerator0)) -> (NewEnumerator0:Exec, NewEnumerator1:Exec, NewEnumerator2:Exec) {is=title K2Node_SwitchEnum}
   n3 = Get!(execute=@n2.NewEnumerator1) -> (then, else, ContentText:Common Text Object Reference)
   > "SetText (Text)"(self=@n3, InText=@n1)
   n4 = Get!(execute=@n2.NewEnumerator2) -> (then, else, ContentRichText:Common Rich Text Block Object Reference)
   > SetText!(execute, self:Rich Text Block Object Reference=@n4, InText:"Text (by ref)"=@n1) -> (then)
  ]}
 SetDialogTitle {nodes [
   n1 = SetDialogTitle!() -> (then, Title:Text) {is=title K2Node_FunctionEntry}
   > n2 = Get!(execute) -> (then, else, ContentTitle:Common Text Object Reference)
   > "SetText (Text)"(self=@n2, InText=@n1)
  ]}
 SetDialogTitleImage {nodes [
   n1 = SetDialogTitleImage!() -> (then, MaterialInstance:Material Instance Object Reference) {is=title K2Node_FunctionEntry}
   > n2 = Branch(Condition=IsValid(Object=@n1))
   > n3 = Get!(execute) -> (then, else, ContentTitleImage:Image Object Reference)
   > SetVisibility(self=@n3, InVisibility=SelfHitTestInvisible)
   > SetBrushFromMaterial(self=@n3, Material=@n1)
   SetVisibility(execute=@n2.else, self=Get ContentTitleImage() -> (ContentTitleImage:Image Object Reference), InVisibility=Collapsed)
  ]}
 SetFooterTextProperties {nodes [
   n1 = SetFooterTextProperties!() -> (then, Data:S UI Text Structure) {is=title K2Node_FunctionEntry}
   n2 = Get FooterText() -> (FooterText:Common Text Object Reference)
   > SetCommonTextProperties!(
    execute
    self
    S_UI_Text:"S UI Text Structure (by ref)"=@n1
    TextWidget:Common Text Object Reference=@n2
    PlacementWidget:Widget Object Reference=@n2
   ) -> (then) {is=title K2Node_CallFunction}
  ]}
 SetImageProperties {nodes [
   n1 = SetImageProperties!() -> (then, Data:"S UI Image Structure (by ref)") {is=title K2Node_FunctionEntry}
   > n2 = Get!(execute) -> (then, else, ContentImage:Image Object Reference)
   n3 = Break S UI Image(S_UI_Image=@n1)
   > SetPlacement(Widget=@n2, Placement=@n3)
   > SetDesiredSizeOverride(self=@n2, DesiredSize=@n3.ImageSize_72_#11)
   > SetBrushFromMaterial(self=@n2, Material=@n3.Material_75_#12)
  ]}
 SetPlacement {nodes [
   n1 = SetPlacement!() -> (then, Widget:Widget Object Reference, Placement:S UI Placement Structure) {is=title K2Node_FunctionEntry}
   > n2 = Cast To GridSlot!(execute, Object=Get Slot(self=@n1)) -> (then, CastFailed, AsGrid Slot:Grid Slot Object Reference) {is=title K2Node_DynamicCast}
   n3 = Break S UI Placement!(S_UI_Placement:"S UI Placement Structure (by ref)"=@n1.Placement) -> (
    HorizontalAlignment_4_#4:EHorizontalAlignment Enum
    VerticalAlignment_5_#26:EVerticalAlignment Enum
    Padding_8_#27:Margin Structure
    Grid_20_#28:S UI Grid Cell Structure
   ) {is=title K2Node_BreakStruct}
   > SetHorizontalAlignment(self:Grid Slot Object Reference=@n2, InHorizontalAlignment=@n3)
   > SetVerticalAlignment(self:Grid Slot Object Reference=@n2, InVerticalAlignment=@n3.VerticalAlignment_5_#26)
   > SetPadding(self:Grid Slot Object Reference=@n2, InPadding=@n3.Padding_8_#27)
   n4 = Break S UI Grid Cell!(S_UI_GridCell:"S UI Grid Cell Structure (by ref)"=@n3.Grid_20_#28) -> (
    Row_15_#4:int
    RowSpan_17_#29:int
    Column_19_#30:int
    ColumnSpan_21_#31:int
    Layer_23_#32:int
    Nudge_26_#33:Vector 2D Structure
   ) {is=title K2Node_BreakStruct}
   > SetRow!(execute, self:Grid Slot Object Reference=@n2, InRow:int=@n4) -> (then)
   > SetRowSpan!(execute, self:Grid Slot Object Reference=@n2, InRowSpan:int=@n4.RowSpan_17_#29) -> (then)
   > SetColumn!(execute, self:Grid Slot Object Reference=@n2, InColumn:int=@n4.Column_19_#30) -> (then)
   > SetColumnSpan!(execute, self:Grid Slot Object Reference=@n2, InColumnSpan:int=@n4.ColumnSpan_21_#31) -> (then)
   > SetLayer!(execute, self:Grid Slot Object Reference=@n2, InLayer:int=@n4.Layer_23_#32) -> (then)
   > SetNudge!(execute, self:Grid Slot Object Reference=@n2, InNudge:Vector 2D Structure=@n4.Nudge_26_#33) -> (then)
  ]}
 SetScrim {nodes [
   n1 = SetScrim!() -> (then, Blur Strength:float=~, Color:Linear Color Structure, Material Instance:Material Instance Object Reference, Enable:bool=~) {is=title K2Node_FunctionEntry}
   > n2 = Get!(execute) -> (then, else, Scrim:WBP UIKit Scrim Object Reference)
   > SetBlur!(execute, self:WBP UIKit Scrim Object Reference=@n2, Blur Strength:float=@n1) -> (then)
   > SetDimColor!(execute, self:WBP UIKit Scrim Object Reference=@n2, Color:Linear Color Structure=@n1.Color) -> (then)
   > SetDimMaterial!(execute, self:WBP UIKit Scrim Object Reference=@n2, Material Instance:Material Instance Object Reference=@n1.Material Instance) -> (then)
   > SetClickCatch!(execute, self:WBP UIKit Scrim Object Reference=@n2, Enable:bool=@n1.Enable) -> (then)
  ]}
 SetSizeBox {nodes [
   n1 = SetSizeBox!() -> (then, Data:"S UI Size Box Structure (by ref)") {is=title K2Node_FunctionEntry}
   > n2 = Sequence!(execute) -> (then_0, then_1, then_2:Exec, then_3:Exec, then_4:Exec, then_5:Exec, then_6:Exec, then_7:Exec)
   n3 = Break S UI Size Box!(S_UI_SizeBox:"S UI Size Box Structure (by ref)"=@n1) -> (
    WidthOverride_107_#34:S UI Bool Float Structure
    HeightOverride_108_#35:S UI Bool Float Structure
    MinDesiredWidth_109_#36:S UI Bool Float Structure
    MinDesiredHeight_110_#37:S UI Bool Float Structure
    MaxDesiredWidth_111_#38:S UI Bool Float Structure
    MaxDesiredHeight_112_#39:S UI Bool Float Structure
    MinAspectRatio_113_#40:S UI Bool Float Structure
    MaxAspectRatio_114_#41:S UI Bool Float Structure
   ) {is=title K2Node_BreakStruct}
   n4 = Break S UI Bool Float!(S_UI_BoolFloat:"S UI Bool Float Structure (by ref)"=@n3) -> (Enable_3_#42:bool=~, Value_4_#43:double=~) {is=title K2Node_BreakStruct}
   > Branch(Condition=@n4)
   n5 = Get SizeBox_Dialog()
   > SetWidthOverride(self=@n5, InWidthOverride=@n4.Value_4_#43)
   n6 = Break S UI Bool Float(S_UI_BoolFloat=@n3.HeightOverride_108_#35) -> (Enable_3_#42=~, Value_4_#43=~)
   Branch(execute=@n2.then_1, Condition=@n6)
   > SetHeightOverride(self=@n5, InHeightOverride=@n6.Value_4_#43)
   n7 = Break S UI Bool Float(S_UI_BoolFloat=@n3.MinDesiredWidth_109_#36) -> (Enable_3_#42=~, Value_4_#43=~)
   Branch(execute=@n2.then_2, Condition=@n7)
   > SetMinDesiredWidth!(execute, self:Size Box Object Reference=@n5, InMinDesiredWidth:float=@n7.Value_4_#43) -> (then)
   n8 = Break S UI Bool Float(S_UI_BoolFloat=@n3.MinDesiredHeight_110_#37) -> (Enable_3_#42=~, Value_4_#43=~)
   Branch(execute=@n2.then_3, Condition=@n8)
   > SetMinDesiredHeight!(execute, self:Size Box Object Reference=@n5, InMinDesiredHeight:float=@n8.Value_4_#43) -> (then)
   n9 = Break S UI Bool Float(S_UI_BoolFloat=@n3.MaxDesiredWidth_111_#38) -> (Enable_3_#42=~, Value_4_#43=~)
   Branch(execute=@n2.then_4, Condition=@n9)
   > SetMaxDesiredWidth!(execute, self:Size Box Object Reference=@n5, InMaxDesiredWidth:float=@n9.Value_4_#43) -> (then)
   n10 = Break S UI Bool Float(S_UI_BoolFloat=@n3.MaxDesiredHeight_112_#39) -> (Enable_3_#42=~, Value_4_#43=~)
   Branch(execute=@n2.then_5, Condition=@n10)
   > SetMaxDesiredHeight!(execute, self:Size Box Object Reference=@n5, InMaxDesiredHeight:float=@n10.Value_4_#43) -> (then)
   n11 = Break S UI Bool Float(S_UI_BoolFloat=@n3.MinAspectRatio_113_#40) -> (Enable_3_#42=~, Value_4_#43=~)
   Branch(execute=@n2.then_6, Condition=@n11)
   > SetMinAspectRatio!(execute, self:Size Box Object Reference=@n5, InMinAspectRatio:float=@n11.Value_4_#43) -> (then)
   n12 = Break S UI Bool Float(S_UI_BoolFloat=@n3.MaxAspectRatio_114_#41) -> (Enable_3_#42=~, Value_4_#43=~)
   Branch(execute=@n2.then_7, Condition=@n12)
   > SetMaxAspectRatio!(execute, self:Size Box Object Reference=@n5, InMaxAspectRatio:float=@n12.Value_4_#43) -> (then)
  ]}
 SetTextProperties {nodes [
   n1 = SetTextProperties!() -> (then, Data:S UI Text Structure) {is=title K2Node_FunctionEntry}
   > n2 = Switch on E_UI_DialogTextType(Selection=Get UseText() -> (UseText=NewEnumerator0))
   SetCommonTextProperties(
    execute=@n2.NewEnumerator1
    S_UI_Text=@n1
    TextWidget=Get ContentText() -> (ContentText:Common Text Object Reference)
    PlacementWidget=Get ScrollBox_Text() -> (ScrollBox_Text:Scroll Box Object Reference)
   )
   n3 = Get!(execute=@n2.NewEnumerator2) -> (then, else, ContentRichText:Common Rich Text Block Object Reference)
   n4 = Break S UI Text(S_UI_Text=@n1) -> (
    LineHeightPercentage_55_#18=~
    AutoWrapText_58_#19=~
    MobileFontSizeMultiplier_73_#20=~
   )
   > SetPlacement(Widget=Get ScrollBox_Text(), Placement=@n4) {is=title K2Node_CallFunction}
   > SetDefaultFont!(
    execute
    self:Rich Text Block Object Reference=@n3
    InFontInfo:Slate Font Info Structure=Apply Mobile Font Size Multiplier to Slate Font!(
     SlateFontInfo:"Slate Font Info Structure (by ref)"=@n4.Font_45_#14
     MobileFontSizeMultiplier:double=@n4.MobileFontSizeMultiplier_73_#20
    ) -> (ReturnValue:Slate Font Info Structure) {is=title K2Node_MacroInstance}
   ) -> (then)
   > SetDefaultColorAndOpacity!(execute, self:Rich Text Block Object Reference=@n3, InColorAndOpacity:Slate Color Structure=@n4.Color_65_#15) -> (then)
   > SetTextTransformPolicy!(
    execute
    self:Rich Text Block Object Reference=@n3
    InTransformPolicy:ETextTransformPolicy Enum=@n4.TransformPolicy_49_#16
   ) -> (then)
   > Set Justification(Justification=@n4.Justification_52_#17, self=@n3) -> (Output_Get=Left)
   > Set LineHeightPercentage!(
    execute
    LineHeightPercentage:float=@n4.LineHeightPercentage_55_#18
    self:Text Layout Widget Object Reference=@n3
   ) -> (then, Output_Get:float)
   > SetAutoWrapText(self:Rich Text Block Object Reference=@n3, InAutoTextWrap=@n4.AutoWrapText_58_#19)
  ]}
 SetTitleImageProperties {nodes [
   n1 = SetTitleImageProperties!() -> (then, Image:Image Object Reference) {is=title K2Node_FunctionEntry}
   n2 = SlotAsHorizontalBoxSlot!(Widget:Widget Object Reference=@n1) -> (ReturnValue:Horizontal Box Slot Object Reference)
   > SetHorizontalAlignment(
    self:Horizontal Box Slot Object Reference=@n2
    InHorizontalAlignment=Get TitleImageHorizontalAligment() -> (TitleImageHorizontalAligment:EHorizontalAlignment Enum=HAlign_Fill)
   )
   > SetVerticalAlignment(
    self:Horizontal Box Slot Object Reference=@n2
    InVerticalAlignment=Get TitleImageVerticalAligment() -> (TitleImageVerticalAligment:EVerticalAlignment Enum=VAlign_Fill)
   )
   > SetPadding(self:Horizontal Box Slot Object Reference=@n2, InPadding=Get TitleImagePadding() -> (TitleImagePadding:Margin Structure))
   > SetSize(self=@n2, InSize=Get TitleImageSizeRule() -> (TitleImageSizeRule:Slate Child Size Structure))
   > SetBrush(
    self=Get Image() -> (Image:Image Object Reference)
    InBrush=Make Slate Brush!(ImageSize_X:double=Get TitleImageSizeX() -> (TitleImageSizeX:double), ImageSize_Y:double=Get TitleImageSizeY() -> (TitleImageSizeY:double)) -> (SlateBrush:Slate Brush Structure) {is=title K2Node_MakeStruct}
   )
  ]}
 SetTitleProperties {nodes [
   n1 = SetTitleProperties!() -> (then, Data:S UI Text Structure) {is=title K2Node_FunctionEntry}
   > SetCommonTextProperties(
    S_UI_Text=@n1
    TextWidget=Get ContentTitle() -> (ContentTitle:Common Text Object Reference)
    PlacementWidget=Get TitleContainer() -> (TitleContainer:Horizontal Box Object Reference)
   )
   > SetTitleImageProperties!(execute, self, Image:Image Object Reference=Get ContentTitleImage()) -> (then) {is=title K2Node_CallFunction}
   > SetTitleTextProperties!(execute, self, Widget:Widget Object Reference=Get ContentTitle()) -> (then)
  ]}
 SetTitleTextProperties {nodes [
   n1 = SetTitleTextProperties!() -> (then, Widget:Widget Object Reference) {is=title K2Node_FunctionEntry}
   n2 = SlotAsHorizontalBoxSlot(Widget=@n1)
   > SetHorizontalAlignment(self=@n2, InHorizontalAlignment=Get TitleTextHorizontalAligment() -> (TitleTextHorizontalAligment:EHorizontalAlignment Enum=HAlign_Fill))
   > SetVerticalAlignment(self=@n2, InVerticalAlignment=Get TitleTextVerticalAligment() -> (TitleTextVerticalAligment:EVerticalAlignment Enum=VAlign_Fill))
   > SetPadding(self=@n2, InPadding=Get TitleTextPadding() -> (TitleTextPadding:Margin Structure))
   > SetSize(self=@n2, InSize=Get TitleTextSizeRule() -> (TitleTextSizeRule:Slate Child Size Structure))
  ]}
 TryAddBackground {nodes [
   Inputs!() -> (execute)
   > n1 = Branch(Condition=IsValid(Object=Get ContentBackground() -> (ContentBackground:Image Object Reference)))
   n2 = Get UseBackground() -> (UseBackground:bool)
   > n3 = Branch(Condition=@n2)
   Outputs(then=[@n8.then, @n4.then, @n5.else, @n3.then])
   RemoveFromParent(execute=@n3.else, self=Get ContentBackground())
   > n4 = Set ContentBackground(ContentBackground:Image Object Reference) -> (Output_Get:Image Object Reference)
   n5 = Branch(execute=@n1.else, Condition=@n2)
   > n6 = Construct Object from Class(Class=&Image) -> (ReturnValue:Image Object Reference)
   > n7 = Set ContentBackground(ContentBackground=@n6)
   > n8 = AddChildToGrid!(
    execute
    self:Grid Panel Object Reference=Get Main() -> (Main:Grid Panel Object Reference)
    Content:Widget Object Reference=@n7
    InRow:int=0
    InColumn:int=0
   ) -> (then, ReturnValue:Grid Slot Object Reference)
  ]}
 TryAddBackplate {nodes [
   Inputs()
   > n1 = Branch(Condition=IsValid(Object=Get ContentBackplate() -> (ContentBackplate:WBP UIKit Backplate Base Object Reference)))
   n2 = Get UseBackplate() -> (UseBackplate:bool)
   > n3 = Branch(Condition=@n2)
   Outputs(then=[@n11.then, @n14.then, @n4.then, @n5.else, @n3.then])
   RemoveFromParent(execute=@n3.else, self=Get ContentBackplate())
   > n4 = Set ContentBackplate(ContentBackplate:WBP UIKit Backplate Base Object Reference) -> (Output_Get:WBP UIKit Backplate Base Object Reference)
   n5 = Branch(execute=@n1.else, Condition=@n2)
   > n6 = Branch(Condition=IsNintendoSwitch!() -> (ReturnValue:bool))
   n7 = GetOwningPlayer()
   > n8 = Create Widget!(execute, Class=/FortUILibrary/UIKit/Backplates/WBP_UIKit_Backplate_Fallback.WBP_UIKit_Backplate_Fallback_C, OwningPlayer=@n7) -> (then, ReturnValue:WBP UIKit Backplate Fallback Object Reference)
   > n9 = Set ContentBackplate(ContentBackplate=@n8)
   n10 = Get Main()
   > n11 = AddChildToGrid(self=@n10, Content=@n9, InRow=0, InColumn=0)
   n12 = Create Widget(execute=@n6.else, Class=/FortUILibrary/UIKit/Backplates/WBP_UIKit_Backplate.WBP_UIKit_Backplate_C, OwningPlayer=@n7) -> (ReturnValue:WBP UIKit Backplate Object Reference)
   > n13 = Set ContentBackplate(ContentBackplate=@n12)
   > n14 = AddChildToGrid(self=@n10, Content=@n13, InRow=0, InColumn=0)
  ]}
 TryAddFooter {nodes [
   Inputs()
   > n1 = Branch(Condition=IsValid(Object=Get FooterText()))
   n2 = Get UseFooter() -> (UseFooter:bool)
   > n3 = Branch(Condition=@n2)
   Outputs(then=[@n8.then, @n4.then, @n5.else, @n3.then])
   RemoveFromParent(execute=@n3.else, self=Get FooterText())
   > n4 = Set FooterText(FooterText:Common Text Object Reference) -> (Output_Get:Common Text Object Reference)
   n5 = Branch(execute=@n1.else, Condition=@n2)
   > n6 = Construct Object from Class!(
    execute
    Class=&CommonTextBlock
    self
    bIsScrollingEnabled:bool=false
    Style:Common Text Style Class Reference
    ScrollStyle:Common Text Scroll Style Class Reference
    ScrollOrientation:EOrientation Enum=Orient_Horizontal
   ) -> (then, ReturnValue:Common Text Object Reference)
   > n7 = Set FooterText(FooterText=@n6)
   > n8 = AddChildToGrid(self=Get Content_FirstColumn() -> (Content_FirstColumn:Grid Panel Object Reference), Content=@n7, InRow=0, InColumn=0)
  ]}
 TryAddImage {nodes [
   Inputs()
   > n1 = Branch(Condition=IsValid(Object=Get ContentImage() -> (ContentImage:Image Object Reference)))
   n2 = Get UseImage() -> (UseImage:bool)
   > n3 = Branch(Condition=@n2)
   Outputs(then=[@n8.then, @n4.then, @n5.else, @n3.then])
   RemoveFromParent(execute=@n3.else, self=Get ContentImage())
   > n4 = Set ContentImage(ContentImage:Image Object Reference) -> (Output_Get:Image Object Reference)
   n5 = Branch(execute=@n1.else, Condition=@n2)
   > n6 = Construct Object from Class!(execute, Class=&Image, self) -> (then, ReturnValue:Image Object Reference)
   > n7 = Set ContentImage(ContentImage=@n6)
   > n8 = AddChildToGrid(self=Get Main(), Content=@n7, InRow=0, InColumn=1)
  ]}
 TryAddScrim {nodes [
   Inputs()
   > n1 = Branch(Condition=IsValid(Object=Get Scrim() -> (Scrim:WBP UIKit Scrim Object Reference)))
   n2 = Get UseScrim() -> (UseScrim:bool)
   > n3 = Branch(Condition=@n2)
   Outputs(then=[@n8.then, @n4.then, @n5.else, @n3.then])
   RemoveFromParent(execute=@n3.else, self=Get Scrim())
   > n4 = Set Scrim(Scrim:WBP UIKit Scrim Object Reference) -> (Output_Get:WBP UIKit Scrim Object Reference)
   n5 = Branch(execute=@n1.else, Condition=@n2)
   > n6 = Create Widget(Class=&/WBP_UIKit_Scrim.WBP_UIKit_Scrim_C, OwningPlayer=GetOwningPlayer()) -> (ReturnValue:WBP UIKit Scrim Object Reference)
   > n7 = Set Scrim(Scrim=@n6)
   > n8 = AddChildToGrid(self=Get UIKitDialog_Root() -> (UIKitDialog_Root:Grid Panel Object Reference), Content=@n7, InRow=0, InColumn=0)
  ]}
 TryAddText {nodes [
   Inputs()
   > Remove!(execute:Exec, Widget:Widget Object Reference=Get ContentText()) -> (then:Exec) {is=title K2Node_MacroInstance}
   > Remove(Widget=Get ContentRichText() -> (ContentRichText:Common Rich Text Block Object Reference))
   > n1 = Switch on E_UI_DialogTextType(Selection=Get UseText() -> (UseText=NewEnumerator0))
   > n2 = Set Visibility(Visibility=Collapsed, self=Get ScrollBox_Text())
   Outputs(then=[@n6.then, @n2.then])
   n3 = Construct Object from Class!(
    execute=@n1.NewEnumerator1
    Class=&CommonTextBlock
    self
    bIsScrollingEnabled:bool=True
    Style:Common Text Style Class Reference
    ScrollStyle:Common Text Scroll Style Class Reference
    ScrollOrientation:EOrientation Enum=Orient_Horizontal
   ) -> (then, ReturnValue:Common Text Object Reference)
   > n4 = Set ContentText(ContentText:Common Text Object Reference=@n3) -> (Output_Get:Common Text Object Reference)
   > n5 = AddChild(self=Get ScrollBox_Text(), Content=@n4)
   n6 = Set Visibility(execute=[@n9.then, @n5.then], Visibility=Visible, self=Get ScrollBox_Text())
   n7 = Construct Object from Class!(
    execute=@n1.NewEnumerator2
    Class=&CommonRichTextBlock
    self
    ScrollStyle
    ScrollOrientation=Orient_Horizontal
    bIsScrollingEnabled=True
   ) -> (then, ReturnValue:Common Rich Text Block Object Reference)
   > n8 = Set ContentRichText(ContentRichText:Common Rich Text Block Object Reference=@n7) -> (Output_Get:Common Rich Text Block Object Reference)
   > n9 = AddChild(self=Get ScrollBox_Text(), Content=@n8)
  ]}
 TryAddTitle {nodes [
   Inputs()
   > Remove(Widget=Get ContentTitle())
   > Remove(Widget=Get ContentTitleImage())
   > n1 = Branch(Condition=Get UseTitle() -> (UseTitle:bool))
   > n2 = Branch(Condition=Get UseTitleImage() -> (UseTitleImage:bool))
   > n3 = Construct Object from Class!(execute, Class=&Image, self) -> (then, ReturnValue:Image Object Reference)
   > n4 = Set ContentTitleImage(ContentTitleImage:Image Object Reference=@n3) -> (Output_Get:Image Object Reference)
   > n5 = AddChildToHorizontalBox!(execute, self:Horizontal Box Object Reference=Get TitleContainer(), Content:Widget Object Reference=@n4) -> (then, ReturnValue:Horizontal Box Slot Object Reference)
   n6 = Construct Object from Class!(
    execute=[@n5.then, @n2.else]
    Class=&CommonTextBlock
    self
    bIsScrollingEnabled:bool=false
    Style:Common Text Style Class Reference
    ScrollStyle:Common Text Scroll Style Class Reference
    ScrollOrientation:EOrientation Enum=Orient_Horizontal
   ) -> (then, ReturnValue:Common Text Object Reference)
   > n7 = Set ContentTitle(ContentTitle:Common Text Object Reference=@n6) -> (Output_Get:Common Text Object Reference)
   > n8 = AddChildToHorizontalBox(self=Get TitleContainer(), Content=@n7)
   Outputs(then=[@n8.then, @n1.else])
  ]}
 TryAutoDeactivate {nodes [
   n1 = Inputs!() -> (execute, ButtonIndex:int)
   > n2 = Get!(execute) -> (then, else, UIKitDialogViewModel:UI Kit Dialog Viewmodel Object Reference)
   n3 = Get Buttons!(self:UI Kit Dialog Viewmodel Object Reference=@n2) -> (Buttons:Array of UI Kit Dialog Button View Model Object References)
   > Branch(Condition=Is Valid Index(TargetArray:Array of UI Kit Dialog Button View Model Object References=@n3, IndexToTest=@n1))
   > Branch(Condition=Get bAutoDeactivate!(self:UI Kit Dialog Button View Model Object Reference="Get (a copy)"(Array:Array of UI Kit Dialog Button View Model Object References=@n3, Dimension 1=@n1) -> (Output:UI Kit Dialog Button View Model Object Reference)) -> (bAutoDeactivate:bool))
   > DeactivationRequested!(execute, self, ShouldBeTrue:bool=true) -> (then) {is=title K2Node_CallFunction}
   > Outputs!(Out:Exec)
  ]}
]
variables [
 BackgroundProperties {
  type=S_UI_Image
  default {
   material=/FortUILibrary/Systems/Dialog/MaterialInstances/MI_UI_DialogBackground_Base
   image Size=<32 32>
   placement {
    horizontal Alignment=HAlign_Fill
    vertical Alignment=VAlign_Fill
    grid {column=0, column Span=2, layer=1, row=0, row Span=0, nudge=<0 0>}
    padding=<ltrb 0 0 0 0>
   }
  }
 }
 BackplateProperties {
  type=S_UI_Backplate
  default {
   blur=100
   brightness=Dark
   corner Radius=Backplate
   placement {
    horizontal Alignment=HAlign_Fill
    vertical Alignment=VAlign_Fill
    grid {column=0, column Span=2, layer=0, row=0, row Span=0, nudge=<0 0>}
    padding=<ltrb 0 0 0 0>
   }
  }
 }
 Blur {type=number, default=10}
 ButtonGroup {type=CommonButtonGroupBase, default=null}
 Buttons {
  type=S_UI_DialogButtons
  default {
   each Button Horizontal Fill=true
   each Button Min Width=0
   each Button Size=Medium
   focuse Button=0
   space Between Buttons=30
   cTA Buttons Types [Loud,Regular]
   group Placement {
    horizontal Alignment=HAlign_Fill
    vertical Alignment=VAlign_Bottom
    grid {column=1, column Span=0, layer=0, row=2, row Span=0, nudge=<0 0>}
    padding=<ltrb 80 64 80 48>
   }
  }
 }
 ButtonWidgets {type="array<WBP_UIKit_ButtonCTA_Base_C>", default []}
 ContentBackground {type=Image, default=null}
 ContentBackplate {type=WBP_UIKit_Backplate_Base_C, default=null}
 ContentImage {type=Image, default=null}
 ContentRichText {type=CommonRichTextBlock, default=null}
 ContentText {type=CommonTextBlock, default=null}
 ContentTitle {type=CommonTextBlock, default=null}
 ContentTitleImage {type=Image, default=null}
 CornerRadiusValues {type=S_UI_BackplateCornerRadiusValues, default {backplate=10, button=0.3, circle=0.5, interactive Component=20, tab=17}}
 DefaultHoldData {type=Class, default=&CommonUI_HoldData_Base_C}
 DefaultInputActionDataTable {type=DataTable, default=null}
 DesignTimePreview_DialogVM {type=Class, default=/FortUILibrary/UIKit/Dialogs/Data/Dialogs/DialogVM_Preview.DialogVM_Preview_C}
 DialogButtonClickedDispatcher {}
 DialogPlacement {
  type=S_UI_Placement
  default {
   horizontal Alignment=HAlign_Center
   vertical Alignment=VAlign_Center
   grid {column=0, column Span=0, layer=1, row=0, row Span=0, nudge=<0 0>}
   padding=<ltrb 0 0 0 0>
  }
 }
 DialogSIze {
  type=S_UI_SizeBox
  default {
   height Override {enable=false, value=800}
   max Aspect Ratio {enable=false, value=0}
   max Desired Height {enable=true, value=800}
   max Desired Width {enable=false, value=0}
   min Aspect Ratio {enable=false, value=0}
   min Desired Height {enable=false, value=0}
   min Desired Width {enable=false, value=0}
   width Override {enable=true, value=1200}
  }
 }
 DimColor {type=LinearColor, default=<rgba 0 0 0 0.2>}
 DimMaterial {type=MaterialInstance, default=null}
 FooterText {type=CommonTextBlock, default=null}
 FooterTextProperties {
  type=S_UI_Text
  default {
   auto Wrap Text=true
   justification=Left
   line Height Percentage=1.35
   mobile Font Size Multiplier=1.3
   transform Policy=None
   wrap Text At=785
   color {colorUseRule=UseColor_Specified, specifiedColor=<rgba 1 0.147 0.1329 1>}
   font {
    bForceMonospaced=false
    bMaterialIsStencil=false
    fontMaterial=null
    fontObject=&HeadingNow
    letterSpacing=1
    monospacedWidth=1
    size=11.25
    skewAmount=0
    typefaceFontName=74 Regular
    outlineSettings {
     bApplyOutlineToDropShadows=false
     bMiteredCorners=false
     bSeparateFillAlpha=false
     outlineMaterial=null
     outlineSize=0
     outlineColor=<rgba 0 0 0 1>
    }
   }
   placement {
    horizontal Alignment=HAlign_Fill
    vertical Alignment=VAlign_Fill
    grid {column=1, column Span=0, layer=0, row=3, row Span=0, nudge=<0 0>}
    padding=<ltrb 80 0 80 56>
   }
  }
 }
 ImageMID {type=MaterialInstanceDynamic}
 ImageProperties {
  type=S_UI_Image
  default {
   material=/FortUILibrary/Systems/Dialog/MaterialInstances/MI_UI_DialogImage
   image Size=<600 0>
   placement {
    horizontal Alignment=HAlign_Fill
    vertical Alignment=VAlign_Fill
    grid {column=1, column Span=0, layer=20, row=0, row Span=1, nudge=<0 0>}
    padding=<ltrb 0 0 0 0>
   }
  }
 }
 ParentActivatableDialog {type=CommonActivatableWidget, default=null}
 Scrim {type=WBP_UIKit_Scrim_C, default=null}
 TextProperties {
  type=S_UI_Text
  default {
   auto Wrap Text=true
   justification=Left
   line Height Percentage=1.2
   mobile Font Size Multiplier=1.3
   transform Policy=None
   wrap Text At=0
   color {colorUseRule=UseColor_Specified, specifiedColor=<rgba 1 1 1 1>}
   font {
    bForceMonospaced=false
    bMaterialIsStencil=false
    fontMaterial=null
    fontObject=&HeadingNow
    letterSpacing=0
    monospacedWidth=1
    size=15.75
    skewAmount=0
    typefaceFontName=74 Regular
    outlineSettings {
     bApplyOutlineToDropShadows=false
     bMiteredCorners=false
     bSeparateFillAlpha=false
     outlineMaterial=null
     outlineSize=0
     outlineColor=<rgba 0 0 0 1>
    }
   }
   placement {
    horizontal Alignment=HAlign_Fill
    vertical Alignment=VAlign_Top
    grid {column=1, column Span=0, layer=0, row=1, row Span=0, nudge=<0 0>}
    padding=<ltrb 80 0 80 0>
   }
  }
 }
 TitleImageHorizontalAligment {type=EHorizontalAlignment, default=HAlign_Center}
 TitleImagePadding {type=Margin, default=<ltrb 0 0 0 0>}
 TitleImageSizeRule {type=SlateChildSize, default {sizeRule=Automatic, value=1}}
 TitleImageSizeX {type=number, default=32}
 TitleImageSizeY {type=number, default=32}
 TitleImageVerticalAligment {type=EVerticalAlignment, default=VAlign_Center}
 TitleProperties {
  type=S_UI_Text
  default {
   auto Wrap Text=true
   justification=Left
   line Height Percentage=1
   mobile Font Size Multiplier=1.3
   transform Policy=ToUpper
   wrap Text At=0
   color {colorUseRule=UseColor_Specified, specifiedColor=<rgba 1 1 1 1>}
   font {
    bForceMonospaced=false
    bMaterialIsStencil=false
    fontMaterial=null
    fontObject=&HeadingNow
    letterSpacing=0
    monospacedWidth=1
    size=25.5
    skewAmount=0
    typefaceFontName=86 Bold Italic
    outlineSettings {
     bApplyOutlineToDropShadows=false
     bMiteredCorners=false
     bSeparateFillAlpha=false
     outlineMaterial=null
     outlineSize=0
     outlineColor=<rgba 0 0 0 1>
    }
   }
   placement {
    horizontal Alignment=HAlign_Fill
    vertical Alignment=VAlign_Top
    grid {column=1, column Span=0, layer=0, row=0, row Span=0, nudge=<0 0>}
    padding=<ltrb 80 80 80 32>
   }
  }
 }
 TitleTextHorizontalAligment {type=EHorizontalAlignment, default=HAlign_Fill}
 TitleTextPadding {type=Margin, default=<ltrb 0 0 0 0>}
 TitleTextSizeRule {type=SlateChildSize, default {sizeRule=Fill, value=1}}
 TitleTextVerticalAligment {type=EVerticalAlignment, default=VAlign_Fill}
 UseBackground {type=boolean}
 UseBackplate {type=boolean, default=true}
 UseClickCatch {type=boolean}
 UseFooter {type=boolean}
 UseImage {type=boolean}
 UseScrim {type=boolean, default=true}
 UseText {type=E_UI_DialogTextType, default=Common Text}
 UseTitle {type=boolean, default=true}
 UseTitleImage {type=boolean}
]
widgets {
 info {inheritedWidgetCount=0, parentClass=&UserWidget, rootWidgetClass=&GridPanel, widgetCount=13}
 widgets [
  UIKitDialog_Root {widgetClassPath=&GridPanel, bIsVariable=true, widget=UIKitDialog_Root, properties {columnFill [1], rowFill [1]}}
  SizeBox_Dialog {
   widgetClassPath=&SizeBox
   bIsVariable=true
   parent=UIKitDialog_Root
   widget=SizeBox_Dialog
   slot {class=GridSlot, properties {horizontalAlignment=HAlign_Center, layer=1, verticalAlignment=VAlign_Center}}
  }
  Main {
   widgetClassPath=&GridPanel
   bIsVariable=true
   parent=SizeBox_Dialog
   widget=Main
   properties {columnFill [1,0], rowFill [1]}
   slot {class=SizeBoxSlot}
  }
  Slot_Background {
   widgetClassPath=&NamedSlot
   bIsVariable=true
   parent=Main
   widget=Slot_Background
   slot {class=GridSlot, properties {columnSpan=2, layer=-2}}
  }
  PostBufferUpdate_Content {
   widgetClassPath=&WBP_CaptureForPostBufferUpdate
   parent=Main
   widget=PostBufferUpdate_Content
   slot {class=GridSlot, properties {columnSpan=2, layer=1}}
  }
  Content_FirstColumn {
   widgetClassPath=&GridPanel
   bIsVariable=true
   parent=Main
   widget=Content_FirstColumn
   properties {columnFill [0,1,0], rowFill [0,1,0,0]}
   slot {class=GridSlot, properties {columnSpan=1, layer=10}}
  }
  ScrollBox_Text {
   widgetClassPath=&ScrollBox
   bIsVariable=true
   parent=Content_FirstColumn
   widget=ScrollBox_Text
   properties {
    bAnimateWheelScrolling=true
    scrollbarPadding=<ltrb 8 0 0 0>
    scrollbarThickness=<12 12>
    widgetBarStyle {
     thickness=12
     draggedThumbImage {
      bIsDynamicallyLoaded=false
      drawAs=RoundedBox
      imageType=FullColor
      mirroring=NoMirror
      resourceName=../../../Engine/Content/Slate/Common/Scrollbar_Thumb.png
      resourceObject=null
      tiling=NoTile
      imageSize=<12 12>
      margin=<ltrb 0.25 0.25 0.25 0.25>
      outlineSettings {
       bUseBrushTransparency=false
       roundingType=FixedRadius
       width=2
       color {colorUseRule=UseColor_Specified, specifiedColor=<rgba 0 0 0 0>}
       cornerRadii=<6 6 6 6>
      }
      tintColor {colorUseRule=UseColor_Specified, specifiedColor=<rgba 1 1 1 1>}
      uVRegion {bIsValid=false, max=<0 0>, min=<0 0>}
     }
     horizontalBackgroundImage {
      bIsDynamicallyLoaded=false
      drawAs=NoDrawType
      imageType=NoImage
      mirroring=NoMirror
      resourceName=None
      resourceObject=null
      tiling=NoTile
      imageSize=<0 0>
      margin=<ltrb 0 0 0 0>
      outlineSettings {
       bUseBrushTransparency=false
       roundingType=HalfHeightRadius
       width=0
       color {colorUseRule=UseColor_Specified, specifiedColor=<rgba 0 0 0 0>}
       cornerRadii=<0 0 0 0>
      }
      tintColor {colorUseRule=UseColor_Specified, specifiedColor=<rgba 1 1 1 1>}
      uVRegion {bIsValid=false, max=<0 0>, min=<0 0>}
     }
     horizontalBottomSlotImage {
      bIsDynamicallyLoaded=false
      drawAs=NoDrawType
      imageType=FullColor
      mirroring=NoMirror
      resourceName=../../../Engine/Content/Slate/Common/Scrollbar_Background_Horizontal.png
      resourceObject=null
      tiling=NoTile
      imageSize=<8 8>
      margin=<ltrb 0 0 0 0>
      outlineSettings {
       bUseBrushTransparency=false
       roundingType=HalfHeightRadius
       width=0
       color {colorUseRule=UseColor_Specified, specifiedColor=<rgba 0 0 0 0>}
       cornerRadii=<0 0 0 0>
      }
      tintColor {colorUseRule=UseColor_Specified, specifiedColor=<rgba 1 1 1 1>}
      uVRegion {bIsValid=false, max=<0 0>, min=<0 0>}
     }
     horizontalTopSlotImage {
      bIsDynamicallyLoaded=false
      drawAs=NoDrawType
      imageType=FullColor
      mirroring=NoMirror
      resourceName=../../../Engine/Content/Slate/Common/Scrollbar_Background_Horizontal.png
      resourceObject=null
      tiling=NoTile
      imageSize=<8 8>
      margin=<ltrb 0 0 0 0>
      outlineSettings {
       bUseBrushTransparency=false
       roundingType=HalfHeightRadius
       width=0
       color {colorUseRule=UseColor_Specified, specifiedColor=<rgba 0 0 0 0>}
       cornerRadii=<0 0 0 0>
      }
      tintColor {colorUseRule=UseColor_Specified, specifiedColor=<rgba 1 1 1 1>}
      uVRegion {bIsValid=false, max=<0 0>, min=<0 0>}
     }
     hoveredThumbImage {
      bIsDynamicallyLoaded=false
      drawAs=RoundedBox
      imageType=FullColor
      mirroring=NoMirror
      resourceName=../../../Engine/Content/Slate/Common/Scrollbar_Thumb.png
      resourceObject=null
      tiling=NoTile
      imageSize=<12 12>
      margin=<ltrb 0.25 0.25 0.25 0.25>
      outlineSettings {
       bUseBrushTransparency=false
       roundingType=FixedRadius
       width=2
       color {colorUseRule=UseColor_Specified, specifiedColor=<rgba 0 0 0 0>}
       cornerRadii=<6 6 6 6>
      }
      tintColor {colorUseRule=UseColor_Specified, specifiedColor=<rgba 1 1 1 1>}
      uVRegion {bIsValid=false, max=<0 0>, min=<0 0>}
     }
     normalThumbImage {
      bIsDynamicallyLoaded=false
      drawAs=RoundedBox
      imageType=FullColor
      mirroring=NoMirror
      resourceName=../../../Engine/Content/Slate/Common/Scrollbar_Thumb.png
      resourceObject=null
      tiling=NoTile
      imageSize=<12 12>
      margin=<ltrb 0.25 0.25 0.25 0.25>
      outlineSettings {
       bUseBrushTransparency=false
       roundingType=FixedRadius
       width=2
       color {colorUseRule=UseColor_Specified, specifiedColor=<rgba 0 0 0 0>}
       cornerRadii=<6 6 6 6>
      }
      tintColor {colorUseRule=UseColor_Specified, specifiedColor=<rgba 1 1 1 0.749>}
      uVRegion {bIsValid=false, max=<0 0>, min=<0 0>}
     }
     verticalBackgroundImage {
      bIsDynamicallyLoaded=false
      drawAs=Box
      imageType=FullColor
      mirroring=NoMirror
      resourceName=None
      resourceObject=/FortUILibrary/UIKit/ScrollBar/MaterialInstances/MI_UIKit_ScrollbarBackground
      tiling=NoTile
      imageSize=<12 12>
      margin=<ltrb 0.5 0.5 0.5 0.5>
      outlineSettings {
       bUseBrushTransparency=false
       roundingType=HalfHeightRadius
       width=0
       color {colorUseRule=UseColor_Specified, specifiedColor=<rgba 0 0 0 0>}
       cornerRadii=<0 0 0 0>
      }
      tintColor {colorUseRule=UseColor_Specified, specifiedColor=<rgba 0.521 0.521 0.521 0.502>}
      uVRegion {bIsValid=false, max=<0 0>, min=<0 0>}
     }
     verticalBottomSlotImage {
      bIsDynamicallyLoaded=false
      drawAs=NoDrawType
      imageType=FullColor
      mirroring=NoMirror
      resourceName=../../../Engine/Content/Slate/Common/Scrollbar_Background_Vertical.png
      resourceObject=null
      tiling=NoTile
      imageSize=<8 8>
      margin=<ltrb 0 0 0 0>
      outlineSettings {
       bUseBrushTransparency=false
       roundingType=HalfHeightRadius
       width=0
       color {colorUseRule=UseColor_Specified, specifiedColor=<rgba 0 0 0 0>}
       cornerRadii=<0 0 0 0>
      }
      tintColor {colorUseRule=UseColor_Specified, specifiedColor=<rgba 1 1 1 1>}
      uVRegion {bIsValid=false, max=<0 0>, min=<0 0>}
     }
     verticalTopSlotImage {
      bIsDynamicallyLoaded=false
      drawAs=NoDrawType
      imageType=FullColor
      mirroring=NoMirror
      resourceName=../../../Engine/Content/Slate/Common/Scrollbar_Background_Vertical.png
      resourceObject=null
      tiling=NoTile
      imageSize=<8 8>
      margin=<ltrb 0 0 0 0>
      outlineSettings {
       bUseBrushTransparency=false
       roundingType=HalfHeightRadius
       width=0
       color {colorUseRule=UseColor_Specified, specifiedColor=<rgba 0 0 0 0>}
       cornerRadii=<0 0 0 0>
      }
      tintColor {colorUseRule=UseColor_Specified, specifiedColor=<rgba 1 1 1 1>}
      uVRegion {bIsValid=false, max=<0 0>, min=<0 0>}
     }
    }
    widgetStyle {
     barThickness=12
     bottomShadowBrush {
      bIsDynamicallyLoaded=false
      drawAs=NoDrawType
      imageType=FullColor
      mirroring=NoMirror
      resourceName=../../../Engine/Content/Slate/Common/ScrollBoxShadowBottom.png
      resourceObject=null
      tiling=NoTile
      imageSize=<16 8>
      margin=<ltrb 0.5 0 0.5 1>
      outlineSettings {
       bUseBrushTransparency=false
       roundingType=HalfHeightRadius
       width=0
       color {colorUseRule=UseColor_Specified, specifiedColor=<rgba 0 0 0 0>}
       cornerRadii=<0 0 0 0>
      }
      tintColor {colorUseRule=UseColor_Specified, specifiedColor=<rgba 1 1 1 1>}
      uVRegion {bIsValid=false, max=<0 0>, min=<0 0>}
     }
     horizontalScrolledContentPadding=<ltrb 0 0 1 0>
     leftShadowBrush {
      bIsDynamicallyLoaded=false
      drawAs=NoDrawType
      imageType=FullColor
      mirroring=NoMirror
      resourceName=../../../Engine/Content/Slate/Common/ScrollBoxShadowLeft.png
      resourceObject=null
      tiling=NoTile
      imageSize=<8 16>
      margin=<ltrb 1 0.5 0 0.5>
      outlineSettings {
       bUseBrushTransparency=false
       roundingType=HalfHeightRadius
       width=0
       color {colorUseRule=UseColor_Specified, specifiedColor=<rgba 0 0 0 0>}
       cornerRadii=<0 0 0 0>
      }
      tintColor {colorUseRule=UseColor_Specified, specifiedColor=<rgba 1 1 1 1>}
      uVRegion {bIsValid=false, max=<0 0>, min=<0 0>}
     }
     rightShadowBrush {
      bIsDynamicallyLoaded=false
      drawAs=NoDrawType
      imageType=FullColor
      mirroring=NoMirror
      resourceName=../../../Engine/Content/Slate/Common/ScrollBoxShadowRight.png
      resourceObject=null
      tiling=NoTile
      imageSize=<8 16>
      margin=<ltrb 0 0.5 1 0.5>
      outlineSettings {
       bUseBrushTransparency=false
       roundingType=HalfHeightRadius
       width=0
       color {colorUseRule=UseColor_Specified, specifiedColor=<rgba 0 0 0 0>}
       cornerRadii=<0 0 0 0>
      }
      tintColor {colorUseRule=UseColor_Specified, specifiedColor=<rgba 1 1 1 1>}
      uVRegion {bIsValid=false, max=<0 0>, min=<0 0>}
     }
     topShadowBrush {
      bIsDynamicallyLoaded=false
      drawAs=NoDrawType
      imageType=FullColor
      mirroring=NoMirror
      resourceName=../../../Engine/Content/Slate/Common/ScrollBoxShadowTop.png
      resourceObject=null
      tiling=NoTile
      imageSize=<16 8>
      margin=<ltrb 0.5 1 0.5 0>
      outlineSettings {
       bUseBrushTransparency=false
       roundingType=HalfHeightRadius
       width=0
       color {colorUseRule=UseColor_Specified, specifiedColor=<rgba 0 0 0 0>}
       cornerRadii=<0 0 0 0>
      }
      tintColor {colorUseRule=UseColor_Specified, specifiedColor=<rgba 1 1 1 1>}
      uVRegion {bIsValid=false, max=<0 0>, min=<0 0>}
     }
     verticalScrolledContentPadding=<ltrb 0 0 0 1>
    }
   }
   slot {class=GridSlot, properties {column=1, row=1, padding=<ltrb 80 0 80 0>}}
  }
  Slot_Content {
   widgetClassPath=&NamedSlot
   bIsVariable=true
   parent=Content_FirstColumn
   widget=Slot_Content
   slot {class=GridSlot, properties {column=1, row=1}}
  }
  ButtonsContainer {
   widgetClassPath=&HorizontalBox
   bIsVariable=true
   parent=Content_FirstColumn
   widget=ButtonsContainer
   slot {class=GridSlot, properties {column=1, horizontalAlignment=HAlign_Left, layer=20, row=2, verticalAlignment=VAlign_Bottom, nudge=<0 24>}}
  }
  TitleContainer {
   widgetClassPath=&HorizontalBox
   bIsVariable=true
   parent=Content_FirstColumn
   widget=TitleContainer
   slot {class=GridSlot, properties {column=1, row=1}}
  }
  Slot_Main {
   widgetClassPath=&NamedSlot
   bIsVariable=true
   parent=Main
   widget=Slot_Main
   slot {class=GridSlot, properties {columnSpan=1, layer=20}}
  }
  Slot_SecondColumn {
   widgetClassPath=&NamedSlot
   bIsVariable=true
   parent=Main
   widget=Slot_SecondColumn
   slot {class=GridSlot, properties {column=1, columnSpan=1, layer=21}}
  }
  Slot_Root {
   widgetClassPath=&NamedSlot
   bIsVariable=true
   parent=UIKitDialog_Root
   widget=Slot_Root
   slot {class=GridSlot, properties {layer=20}}
  }
 ]
}
